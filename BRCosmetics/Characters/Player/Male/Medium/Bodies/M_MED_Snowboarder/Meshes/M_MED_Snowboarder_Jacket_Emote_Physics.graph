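# Physics asset for the male medium Snowboarder's jacket, named for emote use, that sets up seven dynamic bone chains (A–G) for secondary motion.
# Each chain runs from a small master body through three or four box segments, mostly joined by angular limits, with free joints atop A and G.
# Pelvis, spine and thigh capsules form body volumes, with extra thigh shapes.
# B–F masters constrain to spine_03; A and G masters have no parent constraint.
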PhysicsAsset /BRCosmetics/Characters/Player/Male/Medium/Bodies/M_MED_Snowboarder/Meshes/M_MED_Snowboarder_Jacket_Emote_Physics
bodies {
 dyn_jacket_A_01 [{shapeType=Box, extentX=8.8958, extentY=18.6608, extentZ=1.301, center=<4.1945 -0.2139 0.0001>}]
 dyn_jacket_A_02 [{shapeType=Box, extentX=8.5377, extentY=25.5714, extentZ=1.301, center=<4.2558 -0.2139 0.0002>}]
 dyn_jacket_A_03 [{shapeType=Box, extentX=6.2104, extentY=25.0458, extentZ=2.0424, center=<3.0545 -0.2139 0.0001>}]
 dyn_jacket_A_04 [{shapeType=Box, extentX=9.6495, extentY=25.0932, extentZ=2.0424, center=<3.2531 -0.2137 0.0004>}]
 dyn_jacket_A_MSTR [{shapeType=Capsule, length=0.505, radius=0.505}]
 dyn_jacket_B_01 [{shapeType=Box, extentX=9.3719, extentY=23.0592, extentZ=1.9839, center=<4.3342 0 0.0001>}]
 dyn_jacket_B_02 [{shapeType=Box, extentX=7.8199, extentY=22.0378, extentZ=2.178, center=<3.8974 0 0.0001>}]
 dyn_jacket_B_03 [{shapeType=Box, extentX=10.0649, extentY=22.8902, extentZ=1.2411, center=<2.5156 0 0.0001>}]
 dyn_jacket_B_MSTR [{shapeType=Capsule, length=0.505, radius=0.505}]
 dyn_jacket_C_01 [{shapeType=Box, extentX=8.8958, extentY=15, extentZ=3.188, center=<4.3342 0 0.0001>}]
 dyn_jacket_C_02 [{shapeType=Box, extentX=8.1684, extentY=15, extentZ=4.2599, center=<4.0544 0 0.0001>}]
 dyn_jacket_C_03 [{shapeType=Box, extentX=6.8421, extentY=15, extentZ=4.2599, center=<3.4057 0 0.0001>}]
 dyn_jacket_C_MSTR [{shapeType=Box, extentX=1.01, extentY=1.01, extentZ=1.01, rotation=<rot 0 -90 0>}]
 dyn_jacket_D_01 [{shapeType=Box, extentX=9.4841, extentY=15, extentZ=4.6626, center=<4.416 0 0.0001>}]
 dyn_jacket_D_02 [{shapeType=Box, extentX=8.65, extentY=15, extentZ=4.6626, center=<4.416 0 0.0001>}]
 dyn_jacket_D_03 [{shapeType=Box, extentX=6.979, extentY=15, extentZ=4.6626, center=<3.4656 0 0.0001>}]
 dyn_jacket_D_MSTR [{shapeType=Capsule, length=0.505, radius=0.505}]
 dyn_jacket_E_01 [{shapeType=Box, extentX=9.4841, extentY=15, extentZ=3.6029, center=<4.416 0 0.0001>}]
 dyn_jacket_E_02 [{shapeType=Box, extentX=7.66, extentY=15, extentZ=3.3245, center=<3.9183 0 0.0001>}]
 dyn_jacket_E_03 [{shapeType=Box, extentX=7.3017, extentY=15, extentZ=3.3454, center=<3.7186 0 0.0001>}]
 dyn_jacket_E_MSTR [{shapeType=Capsule, length=0.505, radius=0.505}]
 dyn_jacket_F_01 [{shapeType=Box, extentX=9.0359, extentY=13.7282, extentZ=1.4998, center=<4.1716 0 -0.0001>}]
 dyn_jacket_F_02 [{shapeType=Box, extentX=8.3516, extentY=13.7282, extentZ=3.24, center=<4.1471 0 0>}]
 dyn_jacket_F_03 [{shapeType=Box, extentX=6.0721, extentY=24.7282, extentZ=4.24, center=<2.9769 0 0>}]
 dyn_jacket_F_MSTR [{shapeType=Capsule, length=0.505, radius=0.505}]
 dyn_jacket_G_01 [{shapeType=Box, extentX=8.8958, extentY=15, extentZ=3.8401, center=<4.3342 0 0.0001>}]
 dyn_jacket_G_02 [{shapeType=Box, extentX=8.5377, extentY=15, extentZ=3.8401, center=<4.2048 0 0.0001>}]
 dyn_jacket_G_03 [{shapeType=Box, extentX=8.0715, extentY=15, extentZ=3.8401, center=<3.1177 0 0>}]
 dyn_jacket_G_04 [{shapeType=Box, extentX=8.8958, extentY=26, extentZ=4.8401, center=<4.3342 2.6748 -1.2018>, rotation=<rot 0.0001 0 10>}]
 dyn_jacket_G_MSTR [{shapeType=Capsule, length=0.505, radius=0.505}]
 pelvis [{shapeType=Capsule, length=7.4864, radius=13.8389}]
 spine_01 [{shapeType=Capsule, length=5.8208, radius=11.6208}]
 spine_02 [{shapeType=Capsule, length=4.7348, radius=9.131}]
 spine_03 [{shapeType=Capsule, length=5.2245, radius=9.3065}]
 thigh_l [
  {shapeType=Capsule, length=34.5905, radius=11.9334, center=<-19.1309 -1.6566 -0.5508>, rotation=<rot 90 -90 -92.8466>}
  {shapeType=Capsule, length=19.0696, radius=13.6405, center=<-20.5233 -1.1461 10.2676>, rotation=<rot 82.4413 0 0>}
 ]
 thigh_r [
  {shapeType=Capsule, length=33.1802, radius=12.1447, center=<19.9377 0.0004 0.3812>, rotation=<rot 90 0 0>}
  {shapeType=Capsule, length=8.8716, radius=9.4284, center=<10.9066 -1.279 5.8207>, rotation=<rot 90 0 0>}
  {shapeType=Capsule, length=0, radius=19.25, shapeName=thigh_r_capsule, center=<19.382 0 -9.8951>, rotation=<rot -5 -0.0002 0>}
 ]
}
constraints [
 {
  bone1Name=spine_01
  bone2Name=pelvis
  swing1LimitDegrees=45
  swing1Motion=Free
  swing2LimitDegrees=45
  swing2Motion=Free
  twistLimitDegrees=45
  twistMotion=Free
 }
 {
  bone1Name=spine_02
  bone2Name=spine_01
  swing1LimitDegrees=45
  swing1Motion=Free
  swing2LimitDegrees=45
  swing2Motion=Free
  twistLimitDegrees=45
  twistMotion=Free
 }
 {
  bone1Name=spine_03
  bone2Name=spine_02
  swing1LimitDegrees=45
  swing1Motion=Free
  swing2LimitDegrees=45
  swing2Motion=Free
  twistLimitDegrees=45
  twistMotion=Free
 }
 {
  bone1Name=dyn_jacket_B_MSTR
  bone2Name=spine_03
  swing1LimitDegrees=45
  swing1Motion=Free
  swing2LimitDegrees=45
  swing2Motion=Free
  twistLimitDegrees=45
  twistMotion=Free
 }
 {
  bone1Name=dyn_jacket_C_MSTR
  bone2Name=spine_03
  swing1LimitDegrees=45
  swing1Motion=Free
  swing2LimitDegrees=45
  swing2Motion=Free
  twistLimitDegrees=45
  twistMotion=Free
 }
 {
  bone1Name=dyn_jacket_D_MSTR
  bone2Name=spine_03
  swing1LimitDegrees=45
  swing1Motion=Free
  swing2LimitDegrees=45
  swing2Motion=Free
  twistLimitDegrees=45
  twistMotion=Free
 }
 {
  bone1Name=dyn_jacket_E_MSTR
  bone2Name=spine_03
  swing1LimitDegrees=45
  swing1Motion=Free
  swing2LimitDegrees=45
  swing2Motion=Free
  twistLimitDegrees=45
  twistMotion=Free
 }
 {
  bone1Name=dyn_jacket_F_MSTR
  bone2Name=spine_03
  swing1LimitDegrees=45
  swing1Motion=Free
  swing2LimitDegrees=45
  swing2Motion=Free
  twistLimitDegrees=45
  twistMotion=Free
 }
 {bone1Name=dyn_jacket_B_01, bone2Name=dyn_jacket_B_MSTR, swing1LimitDegrees=35, swing2LimitDegrees=45, twistLimitDegrees=30}
 {bone1Name=dyn_jacket_B_02, bone2Name=dyn_jacket_B_01, swing1LimitDegrees=35, swing2LimitDegrees=45, twistLimitDegrees=30}
 {bone1Name=dyn_jacket_B_03, bone2Name=dyn_jacket_B_02, swing1LimitDegrees=35, swing2LimitDegrees=45, twistLimitDegrees=30}
 {bone1Name=dyn_jacket_A_02, bone2Name=dyn_jacket_A_01, swing1LimitDegrees=30, swing2LimitDegrees=30, twistLimitDegrees=30}
 {bone1Name=dyn_jacket_A_03, bone2Name=dyn_jacket_A_02, swing1LimitDegrees=35, swing2LimitDegrees=25, twistLimitDegrees=30}
 {bone1Name=dyn_jacket_A_04, bone2Name=dyn_jacket_A_03, swing1LimitDegrees=35, swing2LimitDegrees=25, twistLimitDegrees=30}
 {bone1Name=dyn_jacket_C_01, bone2Name=dyn_jacket_C_MSTR, swing1LimitDegrees=20, swing2LimitDegrees=45, twistLimitDegrees=20}
 {bone1Name=dyn_jacket_C_02, bone2Name=dyn_jacket_C_01, swing1LimitDegrees=20, swing2LimitDegrees=45, twistLimitDegrees=20}
 {bone1Name=dyn_jacket_C_03, bone2Name=dyn_jacket_C_02, swing1LimitDegrees=20, swing2LimitDegrees=45, twistLimitDegrees=20}
 {bone1Name=dyn_jacket_D_01, bone2Name=dyn_jacket_D_MSTR, swing1LimitDegrees=20, swing2LimitDegrees=45, twistLimitDegrees=20}
 {bone1Name=dyn_jacket_D_02, bone2Name=dyn_jacket_D_01, swing1LimitDegrees=20, swing2LimitDegrees=45, twistLimitDegrees=20}
 {bone1Name=dyn_jacket_E_01, bone2Name=dyn_jacket_E_MSTR, swing1LimitDegrees=20, swing2LimitDegrees=45, twistLimitDegrees=20}
 {bone1Name=dyn_jacket_E_02, bone2Name=dyn_jacket_E_01, swing1LimitDegrees=20, swing2LimitDegrees=45, twistLimitDegrees=20}
 {bone1Name=dyn_jacket_E_03, bone2Name=dyn_jacket_E_02, swing1LimitDegrees=20, swing2LimitDegrees=45, twistLimitDegrees=20}
 {bone1Name=dyn_jacket_G_04, bone2Name=dyn_jacket_G_03, swing1LimitDegrees=30, swing2LimitDegrees=50, twistLimitDegrees=30}
 {bone1Name=dyn_jacket_G_03, bone2Name=dyn_jacket_G_02, swing1LimitDegrees=30, swing2LimitDegrees=50, twistLimitDegrees=30}
 {bone1Name=dyn_jacket_D_03, bone2Name=dyn_jacket_D_02, swing1LimitDegrees=20, swing2LimitDegrees=45, twistLimitDegrees=20}
 {
  bone1Name=thigh_l
  bone2Name=pelvis
  swing1LimitDegrees=45
  swing1Motion=Free
  swing2LimitDegrees=45
  swing2Motion=Free
  twistLimitDegrees=45
  twistMotion=Free
 }
 {
  bone1Name=thigh_r
  bone2Name=pelvis
  swing1LimitDegrees=45
  swing1Motion=Free
  swing2LimitDegrees=45
  swing2Motion=Free
  twistLimitDegrees=45
  twistMotion=Free
 }
 {bone1Name=dyn_jacket_F_01, bone2Name=dyn_jacket_F_MSTR, swing1LimitDegrees=20, swing2LimitDegrees=45, twistLimitDegrees=20}
 {bone1Name=dyn_jacket_F_02, bone2Name=dyn_jacket_F_01, swing1LimitDegrees=20, swing2LimitDegrees=45, twistLimitDegrees=20}
 {bone1Name=dyn_jacket_F_03, bone2Name=dyn_jacket_F_02, swing1LimitDegrees=20, swing2LimitDegrees=45, twistLimitDegrees=20}
 {
  bone1Name=dyn_jacket_A_01
  bone2Name=dyn_jacket_A_MSTR
  swing1LimitDegrees=45
  swing1Motion=Free
  swing2LimitDegrees=45
  swing2Motion=Free
  twistLimitDegrees=45
  twistMotion=Free
 }
 {
  bone1Name=dyn_jacket_G_01
  bone2Name=dyn_jacket_G_MSTR
  swing1LimitDegrees=45
  swing1Motion=Free
  swing2LimitDegrees=45
  swing2Motion=Free
  twistLimitDegrees=45
  twistMotion=Free
 }
 {
  bone1Name=dyn_jacket_G_02
  bone2Name=dyn_jacket_G_01
  swing1LimitDegrees=45
  swing1Motion=Free
  swing2LimitDegrees=45
  swing2Motion=Free
  twistLimitDegrees=45
  twistMotion=Free
 }
]
