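# Physics asset for the male medium Snowboarder body's jacket, giving its dynamic jacket bones simulation shapes.
# Seven chains (A–G) of three or four box segments each hang from small capsule master bodies constrained to spine_04.
# Capsules on the pelvis, spine and thighs approximate the wearer's body, with an extra box on the right thigh.
# Body and master joints leave swing and twist free, unlike the joints within each chain.
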PhysicsAsset /BRCosmetics/Characters/Player/Male/Medium/Bodies/M_MED_Snowboarder/Meshes/M_MED_Snowboarder_Jacket_Physics
bodies {
 dyn_jacket_A_01 [{shapeType=Box, extentX=8.8958, extentY=15, extentZ=2.218, center=<4.3342 0 0.0001>}]
 dyn_jacket_A_02 [{shapeType=Box, extentX=8.5377, extentY=15, extentZ=2.2524, center=<4.2048 0 0.0001>}]
 dyn_jacket_A_03 [{shapeType=Box, extentX=6.0715, extentY=15, extentZ=2.2524, center=<3.1177 0 0>}]
 dyn_jacket_A_04 [{shapeType=Box, extentX=8.8958, extentY=15, extentZ=2.2524, center=<4.3342 0 0.0001>}]
 dyn_jacket_A_MSTR [{shapeType=Capsule, length=0.505, radius=0.505}]
 dyn_jacket_B_01 [{shapeType=Box, extentX=8.8958, extentY=15, extentZ=1.5902, center=<4.3342 0 0.0001>}]
 dyn_jacket_B_02 [{shapeType=Box, extentX=7.8199, extentY=15, extentZ=2.263, center=<3.8974 0 0.0001>}]
 dyn_jacket_B_03 [{shapeType=Box, extentX=6.3162, extentY=15, extentZ=2.091, center=<3.206 0 0.0001>}]
 dyn_jacket_B_MSTR [{shapeType=Capsule, length=0.505, radius=0.505}]
 dyn_jacket_C_01 [{shapeType=Box, extentX=8.8958, extentY=15, extentZ=2.1261, center=<4.3342 0 0.0001>}]
 dyn_jacket_C_02 [{shapeType=Box, extentX=8.1684, extentY=15, extentZ=2.1261, center=<4.0544 0 0.0001>}]
 dyn_jacket_C_03 [{shapeType=Box, extentX=6.8913, extentY=15, extentZ=2.1261, center=<3.4254 0 0.0001>}]
 dyn_jacket_C_MSTR [{shapeType=Capsule, length=0.505, radius=0.505}]
 dyn_jacket_D_01 [{shapeType=Box, extentX=9.4841, extentY=15, extentZ=3.2582, center=<4.416 0 0.0001>}]
 dyn_jacket_D_02 [{shapeType=Box, extentX=7.8546, extentY=15, extentZ=3.2582, center=<4.246 0 0.0001>}]
 dyn_jacket_D_03 [{shapeType=Box, extentX=6.5294, extentY=15, extentZ=3.2582, center=<3.3963 0 0.0001>}]
 dyn_jacket_D_MSTR [{shapeType=Capsule, length=0.505, radius=0.505}]
 dyn_jacket_E_01 [{shapeType=Box, extentX=9.4841, extentY=15, extentZ=2.169, center=<4.416 0 0.0001>}]
 dyn_jacket_E_02 [{shapeType=Box, extentX=8.0553, extentY=15, extentZ=2.169, center=<4.07 0 0.0001>}]
 dyn_jacket_E_03 [{shapeType=Box, extentX=7.0181, extentY=15, extentZ=2.3266, center=<3.703 0 -0.3464>}]
 dyn_jacket_E_MSTR [{shapeType=Capsule, length=0.505, radius=0.505}]
 dyn_jacket_F_01 [{shapeType=Box, extentX=8.6894, extentY=13.4056, extentZ=1.6992, center=<3.9848 0 0.0001>}]
 dyn_jacket_F_02 [{shapeType=Box, extentX=7.5922, extentY=13.5762, extentZ=2.2832, center=<3.6948 0 -0.0001>}]
 dyn_jacket_F_03 [{shapeType=Box, extentX=5.363, extentY=14.1933, extentZ=2.3432, center=<2.7326 0 0>}]
 dyn_jacket_F_MSTR [{shapeType=Capsule, length=0.505, radius=0.505}]
 dyn_jacket_G_01 [{shapeType=Box, extentX=8.8958, extentY=15, extentZ=2.7409, center=<4.3342 0 0.0001>}]
 dyn_jacket_G_02 [{shapeType=Box, extentX=8.5377, extentY=15, extentZ=2.7409, center=<4.2048 0 0.0001>}]
 dyn_jacket_G_03 [{shapeType=Box, extentX=6.0715, extentY=15, extentZ=2.7409, center=<3.1177 0 0>}]
 dyn_jacket_G_04 [{shapeType=Box, extentX=6.2332, extentY=15, extentZ=2.7409, center=<3.1385 0 0.0002>}]
 dyn_jacket_G_MSTR [{shapeType=Capsule, length=0.505, radius=0.505}]
 pelvis [{shapeType=Capsule, length=9.5677, radius=13.0365}]
 spine_01 [{shapeType=Capsule, length=5.8208, radius=11.6208}]
 spine_02 [{shapeType=Capsule, length=8.7017, radius=10.3388}]
 spine_03 [{shapeType=Capsule, length=9.1914, radius=10.5143}]
 spine_04 [{shapeType=Capsule, length=10.154, radius=9.5517}]
 thigh_l [{shapeType=Capsule, length=33.974, radius=10.8958, center=<-19.2133 -0.0002 -1.6608>, rotation=<rot 90 0 0>}]
 thigh_r [
  {shapeType=Capsule, length=34.12, radius=10.7498, center=<19.9377 0.0004 0.3812>, rotation=<rot 90 0 0>}
  {shapeType=Box, extentX=22.1205, extentY=12.2452, extentZ=7.9911, center=<13.4598 0.4598 8.1425>, rotation=<rot -0.1578 3.7043 -0.3437>}
 ]
}
constraints [
 {
  bone1Name=spine_01
  bone2Name=pelvis
  swing1LimitDegrees=45
  swing1Motion=Free
  swing2LimitDegrees=45
  swing2Motion=Free
  twistLimitDegrees=45
  twistMotion=Free
 }
 {
  bone1Name=spine_02
  bone2Name=spine_01
  swing1LimitDegrees=45
  swing1Motion=Free
  swing2LimitDegrees=45
  swing2Motion=Free
  twistLimitDegrees=45
  twistMotion=Free
 }
 {
  bone1Name=spine_03
  bone2Name=spine_02
  swing1LimitDegrees=45
  swing1Motion=Free
  swing2LimitDegrees=45
  swing2Motion=Free
  twistLimitDegrees=45
  twistMotion=Free
 }
 {
  bone1Name=spine_04
  bone2Name=spine_03
  swing1LimitDegrees=45
  swing1Motion=Free
  swing2LimitDegrees=45
  swing2Motion=Free
  twistLimitDegrees=45
  twistMotion=Free
 }
 {
  bone1Name=dyn_jacket_B_MSTR
  bone2Name=spine_04
  swing1LimitDegrees=45
  swing1Motion=Free
  swing2LimitDegrees=45
  swing2Motion=Free
  twistLimitDegrees=45
  twistMotion=Free
 }
 {
  bone1Name=dyn_jacket_A_MSTR
  bone2Name=spine_04
  swing1LimitDegrees=45
  swing1Motion=Free
  swing2LimitDegrees=45
  swing2Motion=Free
  twistLimitDegrees=45
  twistMotion=Free
 }
 {
  bone1Name=dyn_jacket_C_MSTR
  bone2Name=spine_04
  swing1LimitDegrees=45
  swing1Motion=Free
  swing2LimitDegrees=45
  swing2Motion=Free
  twistLimitDegrees=45
  twistMotion=Free
 }
 {
  bone1Name=dyn_jacket_D_MSTR
  bone2Name=spine_04
  swing1LimitDegrees=45
  swing1Motion=Free
  swing2LimitDegrees=45
  swing2Motion=Free
  twistLimitDegrees=45
  twistMotion=Free
 }
 {
  bone1Name=dyn_jacket_E_MSTR
  bone2Name=spine_04
  swing1LimitDegrees=45
  swing1Motion=Free
  swing2LimitDegrees=45
  swing2Motion=Free
  twistLimitDegrees=45
  twistMotion=Free
 }
 {
  bone1Name=dyn_jacket_F_MSTR
  bone2Name=spine_04
  swing1LimitDegrees=45
  swing1Motion=Free
  swing2LimitDegrees=45
  swing2Motion=Free
  twistLimitDegrees=45
  twistMotion=Free
 }
 {
  bone1Name=dyn_jacket_G_MSTR
  bone2Name=spine_04
  swing1LimitDegrees=45
  swing1Motion=Free
  swing2LimitDegrees=45
  swing2Motion=Free
  twistLimitDegrees=45
  twistMotion=Free
 }
 {bone1Name=dyn_jacket_B_01, bone2Name=dyn_jacket_B_MSTR, swing1LimitDegrees=45, swing2LimitDegrees=45, twistLimitDegrees=45}
 {bone1Name=dyn_jacket_B_02, bone2Name=dyn_jacket_B_01, swing1LimitDegrees=45, swing2LimitDegrees=45, twistLimitDegrees=45}
 {bone1Name=dyn_jacket_B_03, bone2Name=dyn_jacket_B_02, swing1LimitDegrees=45, swing2LimitDegrees=45, twistLimitDegrees=45}
 {bone1Name=dyn_jacket_A_01, bone2Name=dyn_jacket_A_MSTR, swing1LimitDegrees=45, swing2LimitDegrees=45, twistLimitDegrees=45}
 {bone1Name=dyn_jacket_A_02, bone2Name=dyn_jacket_A_01, swing1LimitDegrees=45, swing2LimitDegrees=45, twistLimitDegrees=45}
 {bone1Name=dyn_jacket_A_03, bone2Name=dyn_jacket_A_02, swing1LimitDegrees=45, swing2LimitDegrees=45, twistLimitDegrees=45}
 {bone1Name=dyn_jacket_A_04, bone2Name=dyn_jacket_A_03, swing1LimitDegrees=45, swing2LimitDegrees=45, twistLimitDegrees=45}
 {bone1Name=dyn_jacket_C_01, bone2Name=dyn_jacket_C_MSTR, swing1LimitDegrees=45, swing2LimitDegrees=45, twistLimitDegrees=45}
 {bone1Name=dyn_jacket_C_02, bone2Name=dyn_jacket_C_01, swing1LimitDegrees=45, swing2LimitDegrees=45, twistLimitDegrees=45}
 {bone1Name=dyn_jacket_C_03, bone2Name=dyn_jacket_C_02, swing1LimitDegrees=45, swing2LimitDegrees=45, twistLimitDegrees=45}
 {bone1Name=dyn_jacket_D_01, bone2Name=dyn_jacket_D_MSTR, swing1LimitDegrees=45, swing2LimitDegrees=45, twistLimitDegrees=45}
 {bone1Name=dyn_jacket_D_02, bone2Name=dyn_jacket_D_01, swing1LimitDegrees=45, swing2LimitDegrees=45, twistLimitDegrees=45}
 {bone1Name=dyn_jacket_E_01, bone2Name=dyn_jacket_E_MSTR, swing1LimitDegrees=45, swing2LimitDegrees=45, twistLimitDegrees=45}
 {bone1Name=dyn_jacket_E_02, bone2Name=dyn_jacket_E_01, swing1LimitDegrees=45, swing2LimitDegrees=45, twistLimitDegrees=45}
 {bone1Name=dyn_jacket_E_03, bone2Name=dyn_jacket_E_02, swing1LimitDegrees=45, swing2LimitDegrees=45, twistLimitDegrees=45}
 {bone1Name=dyn_jacket_G_01, bone2Name=dyn_jacket_G_MSTR, swing1LimitDegrees=45, swing2LimitDegrees=45, twistLimitDegrees=45}
 {bone1Name=dyn_jacket_G_02, bone2Name=dyn_jacket_G_01, swing1LimitDegrees=45, swing2LimitDegrees=45, twistLimitDegrees=45}
 {bone1Name=dyn_jacket_G_04, bone2Name=dyn_jacket_G_03, swing1LimitDegrees=45, swing2LimitDegrees=45, twistLimitDegrees=45}
 {bone1Name=dyn_jacket_G_03, bone2Name=dyn_jacket_G_02, swing1LimitDegrees=45, swing2LimitDegrees=45, twistLimitDegrees=45}
 {bone1Name=dyn_jacket_D_03, bone2Name=dyn_jacket_D_02, swing1LimitDegrees=45, swing2LimitDegrees=45, twistLimitDegrees=45}
 {
  bone1Name=thigh_l
  bone2Name=pelvis
  swing1LimitDegrees=45
  swing1Motion=Free
  swing2LimitDegrees=45
  swing2Motion=Free
  twistLimitDegrees=45
  twistMotion=Free
 }
 {
  bone1Name=thigh_r
  bone2Name=pelvis
  swing1LimitDegrees=45
  swing1Motion=Free
  swing2LimitDegrees=45
  swing2Motion=Free
  twistLimitDegrees=45
  twistMotion=Free
 }
 {bone1Name=dyn_jacket_F_01, bone2Name=dyn_jacket_F_MSTR, swing1LimitDegrees=45, swing2LimitDegrees=45, twistLimitDegrees=45}
 {bone1Name=dyn_jacket_F_02, bone2Name=dyn_jacket_F_01, swing1LimitDegrees=45, swing2LimitDegrees=45, twistLimitDegrees=45}
 {bone1Name=dyn_jacket_F_03, bone2Name=dyn_jacket_F_02, swing1LimitDegrees=45, swing2LimitDegrees=45, twistLimitDegrees=45}
]
