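# Physics asset for the necklace on the medium male Snowboarder player body, with bodies for its dynamic necklace and collar bones.
# Two four-link capsule chains hang from neck_01, each joint allowing far less twist than swing.
# Left and right collar boxes constrain to spine_05.
# Upper-spine and neck capsules, plus thin angled boxes on the spine, form the torso shapes, joined by free angular constraints.
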PhysicsAsset /BRCosmetics/Characters/Player/Male/Medium/Bodies/M_MED_Snowboarder/Meshes/M_MED_Snowboarder_Necklace_Physics
bodies {
 dyn_l_collar [{shapeType=Box, extentX=2, extentY=2, extentZ=2}]
 dyn_necklace_01 [{shapeType=Capsule, length=12, radius=1, shapeName=dyn_necklace_01_capsule, center=<3.6053 0 -0.0119>, rotation=<rot 0 0 -90>}]
 dyn_necklace_02 [{shapeType=Capsule, length=8, radius=1, shapeName=dyn_necklace_02_capsule, center=<3.2176 0.2189 0.2593>, rotation=<rot 0 0 -90>}]
 dyn_necklace_03 [{shapeType=Capsule, length=6, radius=1, shapeName=dyn_necklace_03_capsule, center=<3.3653 -0.041 0.4219>, rotation=<rot 0 0 -90>}]
 dyn_necklace_04 [{shapeType=Capsule, length=3.3916, radius=1.913, shapeName=dyn_necklace_04_capsule, center=<3 0 0>, rotation=<rot 90 0 0>}]
 dyn_necklace_a_01 [{shapeType=Capsule, length=12, radius=1, shapeName=dyn_necklace_a_01_capsule, center=<3.6053 0 -0.0119>, rotation=<rot 0 0 -90>}]
 dyn_necklace_a_02 [{shapeType=Capsule, length=8, radius=1, shapeName=dyn_necklace_a_02_capsule, center=<3.2176 0.2189 0.2593>, rotation=<rot 0 0 -90>}]
 dyn_necklace_a_03 [{shapeType=Capsule, length=6, radius=1, shapeName=dyn_necklace_a_03_capsule, center=<3.3941 0.0673 0.4171>, rotation=<rot 0 0 -90>}]
 dyn_necklace_a_04 [{shapeType=Capsule, length=3.1249, radius=2.0752, shapeName=dyn_necklace_a_04_capsule, center=<3 0 0>, rotation=<rot 90 0 0>}]
 dyn_r_collar [{shapeType=Box, extentX=2, extentY=2, extentZ=2}]
 neck_01 [{shapeType=Capsule, length=4.5673, radius=4.9347}]
 spine_04 [
  {shapeType=Capsule, length=16.1114, radius=12.2072, center=<2.2281 0.0276 0>}
  {shapeType=Box, extentX=37.0489, extentY=7.5, extentZ=2.8985, center=<5.0189 12.6728 8.1798>, rotation=<rot 4.1693 -13.5287 -8.0153>}
 ]
 spine_05 [
  {shapeType=Capsule, length=17.8665, radius=10.4521}
  {shapeType=Box, extentX=41.2526, extentY=10.7008, extentZ=2.8985, center=<-4.3089 14.2079 -7.035>, rotation=<rot -9.38 -18.734 15.9689>}
 ]
}
constraints [
 {
  bone1Name=spine_05
  bone2Name=spine_04
  swing1LimitDegrees=45
  swing1Motion=Free
  swing2LimitDegrees=45
  swing2Motion=Free
  twistLimitDegrees=45
  twistMotion=Free
 }
 {bone1Name=dyn_l_collar, bone2Name=spine_05, swing1LimitDegrees=30, swing2LimitDegrees=30, twistLimitDegrees=30}
 {bone1Name=dyn_r_collar, bone2Name=spine_05, swing1LimitDegrees=30, swing2LimitDegrees=30, twistLimitDegrees=30}
 {
  bone1Name=neck_01
  bone2Name=spine_05
  swing1LimitDegrees=45
  swing1Motion=Free
  swing2LimitDegrees=45
  swing2Motion=Free
  twistLimitDegrees=45
  twistMotion=Free
 }
 {bone1Name=dyn_necklace_01, bone2Name=neck_01, swing1LimitDegrees=45, swing2LimitDegrees=45, twistLimitDegrees=5}
 {bone1Name=dyn_necklace_02, bone2Name=dyn_necklace_01, swing1LimitDegrees=45, swing2LimitDegrees=45, twistLimitDegrees=5}
 {bone1Name=dyn_necklace_03, bone2Name=dyn_necklace_02, swing1LimitDegrees=45, swing2LimitDegrees=45, twistLimitDegrees=5}
 {bone1Name=dyn_necklace_04, bone2Name=dyn_necklace_03, swing1LimitDegrees=45, swing2LimitDegrees=45, twistLimitDegrees=5}
 {bone1Name=dyn_necklace_a_01, bone2Name=neck_01, swing1LimitDegrees=45, swing2LimitDegrees=45, twistLimitDegrees=5}
 {bone1Name=dyn_necklace_a_02, bone2Name=dyn_necklace_a_01, swing1LimitDegrees=45, swing2LimitDegrees=45, twistLimitDegrees=5}
 {bone1Name=dyn_necklace_a_03, bone2Name=dyn_necklace_a_02, swing1LimitDegrees=45, swing2LimitDegrees=45, twistLimitDegrees=5}
 {bone1Name=dyn_necklace_a_04, bone2Name=dyn_necklace_a_03, swing1LimitDegrees=45, swing2LimitDegrees=45, twistLimitDegrees=5}
]
