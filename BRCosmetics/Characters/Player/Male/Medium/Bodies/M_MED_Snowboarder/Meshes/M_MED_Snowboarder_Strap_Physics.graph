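# Physics asset for the straps on the male medium Snowboarder cosmetic body, set up for secondary motion on its dynamic bones.
# Box bodies cover the left, right and back straps and front and back jacket panels, with capsules on the two upper spine bones.
# Each strap has angular limits against a tiny master capsule.
# The masters link with free rotation to spine_05 (back), back jacket (left) or front jacket (right).
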PhysicsAsset /BRCosmetics/Characters/Player/Male/Medium/Bodies/M_MED_Snowboarder/Meshes/M_MED_Snowboarder_Strap_Physics
bodies {
 dyn_backStrap [{shapeType=Box, extentX=3.9612, extentY=4.439, extentZ=1.2839, center=<2.2981 0 0.1842>}]
 dyn_backStrap_MSTR [{shapeType=Capsule, length=1, radius=1}]
 dyn_jacket_B_02 [{shapeType=Box, extentX=13.4705, extentY=17.6435, extentZ=3.826, center=<6.4995 -0.0001 0>}]
 dyn_jacket_F_02 [{shapeType=Box, extentX=13.4705, extentY=17.6435, extentZ=3.826, center=<6.4995 -0.0001 0>}]
 dyn_l_Strap [{shapeType=Box, extentX=4.119, extentY=4.1786, extentZ=1.0707, center=<2.144 0.0004 -0.049>, rotation=<rot -10.2594 0.0129 5.819>}]
 dyn_l_Strap_MSTR [{shapeType=Capsule, length=1, radius=1}]
 dyn_r_Strap [{shapeType=Box, extentX=4.119, extentY=4.1786, extentZ=1.3404, center=<2.2331 -0.0206 0.4491>, rotation=<rot -10.197 -1.4207 -2.6377>}]
 dyn_r_Strap_MSTR [{shapeType=Capsule, length=1, radius=1}]
 spine_04 [{shapeType=Capsule, length=16.2078, radius=15.3915}]
 spine_05 [{shapeType=Capsule, length=17.8033, radius=9.9156, center=<1 0 0>}]
}
constraints [
 {bone1Name=dyn_r_Strap, bone2Name=dyn_r_Strap_MSTR, swing1LimitDegrees=20, swing2LimitDegrees=60, twistLimitDegrees=5}
 {bone1Name=dyn_l_Strap, bone2Name=dyn_l_Strap_MSTR, swing1LimitDegrees=20, swing2LimitDegrees=60, twistLimitDegrees=5}
 {bone1Name=dyn_backStrap, bone2Name=dyn_backStrap_MSTR, swing1LimitDegrees=20, swing2LimitDegrees=60, twistLimitDegrees=5}
 {
  bone1Name=dyn_backStrap_MSTR
  bone2Name=spine_05
  swing1LimitDegrees=45
  swing1Motion=Free
  swing2LimitDegrees=45
  swing2Motion=Free
  twistLimitDegrees=45
  twistMotion=Free
 }
 {
  bone1Name=dyn_l_Strap_MSTR
  bone2Name=dyn_jacket_B_02
  swing1LimitDegrees=45
  swing1Motion=Free
  swing2LimitDegrees=45
  swing2Motion=Free
  twistLimitDegrees=45
  twistMotion=Free
 }
 {
  bone1Name=dyn_r_Strap_MSTR
  bone2Name=dyn_jacket_F_02
  swing1LimitDegrees=45
  swing1Motion=Free
  swing2LimitDegrees=45
  swing2Motion=Free
  twistLimitDegrees=45
  twistMotion=Free
 }
 {
  bone1Name=spine_05
  bone2Name=spine_04
  swing1LimitDegrees=45
  swing1Motion=Free
  swing2LimitDegrees=45
  swing2Motion=Free
  twistLimitDegrees=45
  twistMotion=Free
 }
 {
  bone1Name=dyn_jacket_B_02
  bone2Name=spine_05
  swing1LimitDegrees=45
  swing1Motion=Free
  swing2LimitDegrees=45
  swing2Motion=Free
  twistLimitDegrees=45
  twistMotion=Free
 }
 {
  bone1Name=dyn_jacket_F_02
  bone2Name=spine_05
  swing1LimitDegrees=45
  swing1Motion=Free
  swing2LimitDegrees=45
  swing2Motion=Free
  twistLimitDegrees=45
  twistMotion=Free
 }
]
