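# Physics asset for the shoe tabs on the medium male Snowboarder character body, covering the left and right dyn_ tab bones and both feet.
# Each tab has a small box body, and each foot has a capsule body.
# A constraint links each tab to its same-side foot, with one wide swing axis and tight limits on the other swing and twist, giving a hinge-like flap motion.
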PhysicsAsset /BRCosmetics/Characters/Player/Male/Medium/Bodies/M_MED_Snowboarder/Meshes/M_MED_Snowboarder_shoeTab_Physics
bodies {
 dyn_l_shoeTab [{shapeType=Box, extentX=2, extentY=2, extentZ=2, center=<-1.0627 0.0003 -0.6887>}]
 dyn_r_shoeTab [{shapeType=Box, extentX=2, extentY=2, extentZ=2, center=<-1.0627 0.0003 -0.6887>}]
 foot_l [{shapeType=Capsule, length=10.8458, radius=8.1497, center=<-4.1466 8.1029 0.0001>, rotation=<rot 90 0 -90>}]
 foot_r [{shapeType=Capsule, length=10.8458, radius=8.1497, center=<2.5389 -9.3867 0.0005>, rotation=<rot 90 -26.565 -116.565>}]
}
constraints [
 {bone1Name=dyn_l_shoeTab, bone2Name=foot_l, swing1LimitDegrees=5, swing2LimitDegrees=45, twistLimitDegrees=5}
 {bone1Name=dyn_r_shoeTab, bone2Name=foot_r, swing1LimitDegrees=5, swing2LimitDegrees=45, twistLimitDegrees=5}
]
