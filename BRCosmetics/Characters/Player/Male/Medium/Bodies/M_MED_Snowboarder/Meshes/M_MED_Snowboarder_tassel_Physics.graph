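# Physics asset for the Snowboarder male medium body's tassel mesh, setting up simulation for the dangling left and right pouch bone chains.
# Box bodies form a four-segment left chain and a three-segment right chain, linked by angle-limited constraints, with wider swing on the left.
# Each chain starts at a small MSTR capsule constrained to its thigh twist capsule with free swing and twist.
# Calf capsules have no constraints, apparently serving as leg collision shapes.
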PhysicsAsset /BRCosmetics/Characters/Player/Male/Medium/Bodies/M_MED_Snowboarder/Meshes/M_MED_Snowboarder_tassel_Physics
bodies {
 calf_l [{shapeType=Capsule, length=25.8948, radius=6.6487, center=<-19.4938 -2.3711 -0.515>, rotation=<rot 88.4096 180 180>}]
 calf_r [{shapeType=Capsule, length=25.9584, radius=6.5851, center=<19.0735 2.3462 0.1019>, rotation=<rot 88.3676 180 180>}]
 dyn_l_pouch_01 [{shapeType=Box, extentX=7.8096, extentY=3.4017, extentZ=2.057, center=<4.0947 0.0001 0.0786>}]
 dyn_l_pouch_02 [{shapeType=Box, extentX=6.4266, extentY=2.7157, extentZ=1.9747, center=<3.2604 0.0001 0.2391>}]
 dyn_l_pouch_03 [{shapeType=Box, extentX=6.6328, extentY=2.7157, extentZ=1.316, center=<3.3306 0.0001 0>}]
 dyn_l_pouch_04 [{shapeType=Box, extentX=3.5377, extentY=2.7157, extentZ=1.2512, center=<1.7894 0 0>}]
 dyn_l_pouch_MSTR [{shapeType=Capsule, length=1, radius=1}]
 dyn_r_pouch_01 [{shapeType=Box, extentX=4.5319, extentY=2.3845, extentZ=1.9526, center=<2.2429 0 0>}]
 dyn_r_pouch_02 [{shapeType=Box, extentX=4.5319, extentY=2.5312, extentZ=1.9526, center=<2.2367 0 0.1735>, rotation=<rot 2.0434 0 0>}]
 dyn_r_pouch_03 [{shapeType=Box, extentX=3.0165, extentY=3.4022, extentZ=1.9526, center=<1.5004 0 0>}]
 dyn_r_pouch_MSTR [{shapeType=Capsule, length=0.505, radius=0.505}]
 thigh_twist_01_l [{shapeType=Capsule, length=24.1415, radius=8.402, center=<1.653 -0.0002 0.2944>, rotation=<rot 87.3281 0 0>}]
 thigh_twist_01_r [{shapeType=Capsule, length=26.0751, radius=7.4277, center=<-1.827 0 -1.0329>, rotation=<rot 86.6337 0 0>}]
}
constraints [
 {bone1Name=dyn_l_pouch_01, bone2Name=dyn_l_pouch_MSTR, swing1LimitDegrees=80, swing2LimitDegrees=80, twistLimitDegrees=40}
 {bone1Name=dyn_l_pouch_02, bone2Name=dyn_l_pouch_01, swing1LimitDegrees=80, swing2LimitDegrees=80, twistLimitDegrees=40}
 {bone1Name=dyn_l_pouch_03, bone2Name=dyn_l_pouch_02, swing1LimitDegrees=80, swing2LimitDegrees=80, twistLimitDegrees=40}
 {bone1Name=dyn_l_pouch_04, bone2Name=dyn_l_pouch_03, swing1LimitDegrees=80, swing2LimitDegrees=80, twistLimitDegrees=40}
 {bone1Name=dyn_r_pouch_01, bone2Name=dyn_r_pouch_MSTR, swing1LimitDegrees=45, swing2LimitDegrees=45, twistLimitDegrees=45}
 {bone1Name=dyn_r_pouch_02, bone2Name=dyn_r_pouch_01, swing1LimitDegrees=45, swing2LimitDegrees=45, twistLimitDegrees=45}
 {bone1Name=dyn_r_pouch_03, bone2Name=dyn_r_pouch_02, swing1LimitDegrees=45, swing2LimitDegrees=45, twistLimitDegrees=45}
 {
  bone1Name=dyn_l_pouch_MSTR
  bone2Name=thigh_twist_01_l
  swing1LimitDegrees=45
  swing1Motion=Free
  swing2LimitDegrees=45
  swing2Motion=Free
  twistLimitDegrees=45
  twistMotion=Free
 }
 {
  bone1Name=dyn_r_pouch_MSTR
  bone2Name=thigh_twist_01_r
  swing1LimitDegrees=45
  swing1Motion=Free
  swing2LimitDegrees=45
  swing2Motion=Free
  twistLimitDegrees=45
  twistMotion=Free
 }
]
