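# Animation Blueprint for the M_MED_Snowboarder character body, adding secondary motion to its jacket, straps and accessories.
# It copies the attached parent mesh's pose, then chains bone constraints and additive bone offsets driven by crouch, skydive, jump/fall and front-end states.
# RigidBody nodes with separate physics assets simulate necklace, tassel, shoe tabs, straps and jacket.
# Per-bone AnimDynamics angular springs add further jacket motion before the jacket RigidBody pass.
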AnimBlueprint /BRCosmetics/Characters/Player/Male/Medium/Bodies/M_MED_Snowboarder/Meshes/M_MED_Snowboarder_LOWMS_AnimBP
target_skeleton=&/M_MED_Snowboarder_Skeleton
graphs [
 AnimGraph {nodes [
   "Constraint - dyn_l_Strap_MSTR (PPPPP)"!(
    ComponentPose
    bAlphaBoolEnabled:bool=True
    ConstraintWeights_0=0.0508
    ConstraintWeights_1:float=0.1363
    ConstraintWeights_2:float=0.7874
    ConstraintWeights_3:float=0.0177
    ConstraintWeights_4:float=0.0077
   ) -> (Pose) {
    alphaInputType=Bool
    alphaBoolBlend {blendOption=Linear}
    boneToModify {boneName=dyn_l_Strap_MSTR}
    constraintSetup [
     {targetBone {boneName=dyn_jacket_B_03}}
     {targetBone {boneName=dyn_jacket_B_02}}
     {targetBone {boneName=dyn_jacket_B_01}}
     {targetBone {boneName=dyn_jacket_A_03}}
     {targetBone {boneName=dyn_jacket_A_01}}
    ]
    constraintWeights [1,1,1,1,1]
   }
   n1 = Get Crouching() -> (Crouching:bool)
   "Constraint - dyn_jacket_A_01 (RP)"!(ComponentPose, bAlphaBoolEnabled:bool=@n1, ConstraintWeights_0=1, ConstraintWeights_1:float=0) -> (Pose) {
    alphaInputType=Bool
    alphaBoolBlend {blendOption=Linear}
    boneToModify {boneName=dyn_jacket_A_01}
    constraintSetup [{transformType=Rotation, targetBone {boneName=thigh_l}}, {targetBone {boneName=pelvis}}]
    constraintWeights [1,1]
   }
   "Constraint - dyn_jacket_A_02 (RP)"!(ComponentPose, bAlphaBoolEnabled:bool=@n1, ConstraintWeights_0=1, ConstraintWeights_1:float=0) -> (Pose) {
    alphaInputType=Bool
    alphaBoolBlend {blendOption=Linear}
    boneToModify {boneName=dyn_jacket_A_02}
    constraintSetup [{transformType=Rotation, targetBone {boneName=thigh_l}}, {targetBone {boneName=pelvis}}]
    constraintWeights [1,1]
   }
   n2 = "Constraint - dyn_jacket_A_MSTR (PP)"!(
    ComponentPose=Local To Component(LocalPose=Copy Pose From Mesh() {bCopyCurves=true, bUseAttachedParent=true})
    bAlphaBoolEnabled:bool=True
    ConstraintWeights_0=1
    ConstraintWeights_1:float=0
   ) -> (Pose) {
    alphaInputType=Bool
    alphaBoolBlend {blendOption=Linear}
    boneToModify {boneName=dyn_jacket_A_MSTR}
    constraintSetup [{targetBone {boneName=pelvis}}, {targetBone {boneName=spine_05}}]
    constraintWeights [1,1]
   }
   n3 = "Constraint - dyn_jacket_B_MSTR (PP)"!(ComponentPose=@n2, bAlphaBoolEnabled:bool=True, ConstraintWeights_0=1, ConstraintWeights_1:float=0) -> (Pose) {
    alphaInputType=Bool
    alphaBoolBlend {blendOption=Linear}
    boneToModify {boneName=dyn_jacket_B_MSTR}
    constraintSetup [{targetBone {boneName=pelvis}}, {targetBone {boneName=spine_05}}]
    constraintWeights [1,1]
   }
   n4 = "Constraint - dyn_jacket_C_MSTR (PP)"!(ComponentPose=@n3, bAlphaBoolEnabled:bool=True, ConstraintWeights_0=1, ConstraintWeights_1:float=0) -> (Pose) {
    alphaInputType=Bool
    alphaBoolBlend {blendOption=Linear}
    boneToModify {boneName=dyn_jacket_C_MSTR}
    constraintSetup [{targetBone {boneName=pelvis}}, {targetBone {boneName=spine_05}}]
    constraintWeights [1,1]
   }
   n5 = "Constraint - dyn_jacket_D_MSTR (PP)"!(ComponentPose=@n4, bAlphaBoolEnabled:bool=True, ConstraintWeights_0=1, ConstraintWeights_1:float=0) -> (Pose) {
    alphaInputType=Bool
    alphaBoolBlend {blendOption=Linear}
    boneToModify {boneName=dyn_jacket_D_MSTR}
    constraintSetup [{targetBone {boneName=pelvis}}, {targetBone {boneName=spine_05}}]
    constraintWeights [1,1]
   }
   n6 = "Constraint - dyn_jacket_E_MSTR (PP)"!(ComponentPose=@n5, bAlphaBoolEnabled:bool=True, ConstraintWeights_0=1, ConstraintWeights_1:float=0) -> (Pose) {
    alphaInputType=Bool
    alphaBoolBlend {blendOption=Linear}
    boneToModify {boneName=dyn_jacket_E_MSTR}
    constraintSetup [{targetBone {boneName=pelvis}}, {targetBone {boneName=spine_05}}]
    constraintWeights [1,1]
   }
   n7 = "Constraint - dyn_jacket_F_MSTR (PP)"!(ComponentPose=@n6, bAlphaBoolEnabled:bool=True, ConstraintWeights_0=1, ConstraintWeights_1:float=0) -> (Pose) {
    alphaInputType=Bool
    alphaBoolBlend {blendOption=Linear}
    boneToModify {boneName=dyn_jacket_F_MSTR}
    constraintSetup [{targetBone {boneName=pelvis}}, {targetBone {boneName=spine_05}}]
    constraintWeights [1,1]
   }
   n8 = "Constraint - dyn_jacket_G_MSTR (PP)"!(ComponentPose=@n7, bAlphaBoolEnabled:bool=True, ConstraintWeights_0=1, ConstraintWeights_1:float=0) -> (Pose) {
    alphaInputType=Bool
    alphaBoolBlend {blendOption=Linear}
    boneToModify {boneName=dyn_jacket_G_MSTR}
    constraintSetup [{targetBone {boneName=pelvis}}, {targetBone {boneName=spine_05}}]
    constraintWeights [1,1]
   }
   n9 = "Constraint - dyn_jacket_G_MSTR (PP)"(ComponentPose=@n8, bAlphaBoolEnabled=True, ConstraintWeights_0=1, ConstraintWeights_1=0) {
    alphaInputType=Bool
    alphaBoolBlend {blendOption=Linear}
    boneToModify {boneName=dyn_jacket_G_MSTR}
    constraintSetup [{targetBone {boneName=pelvis}}, {targetBone {boneName=spine_05}}]
    constraintWeights [1,1]
   }
   n10 = Get Crouching()
   n11 = "Transform (Modify) Bone - Bone: dyn_jacket_A_MSTR"!(ComponentPose=@n9, bAlphaBoolEnabled:bool=@n10, Translation="0,3,-12", Rotation="0,-20,0", Scale="0.5,0.5,0.5") -> (Pose) {
    alphaInputType=Bool
    rotationMode=BMM_Additive
    scaleMode=BMM_Additive
    translationMode=BMM_Additive
    alphaBoolBlend {blendInTime=0.1, blendOption=Linear, blendOutTime=0.1}
    boneToModify {boneName=dyn_jacket_A_MSTR}
    rotation {yaw=-20}
    scale=<0.5 0.5 0.5>
    translation {y=3, z=-12}
   }
   n12 = "Transform (Modify) Bone - Bone: dyn_jacket_B_MSTR"!(ComponentPose=@n11, bAlphaBoolEnabled:bool=@n10, Translation="0,3,0", Rotation="50,20,0", Scale="1,1,1") -> (Pose) {
    alphaInputType=Bool
    rotationMode=BMM_Additive
    scaleMode=BMM_Additive
    translationMode=BMM_Additive
    alphaBoolBlend {blendInTime=0.1, blendOption=Linear, blendOutTime=0.1}
    boneToModify {boneName=dyn_jacket_B_MSTR}
    rotation {pitch=50, yaw=20}
    translation {y=3}
   }
   n13 = "Transform (Modify) Bone - Bone: dyn_jacket_A_01"!(ComponentPose=@n12, bAlphaBoolEnabled:bool=@n10, Translation="0,0,0", Rotation="20,0.000,-35", Scale="0.5,0.5,0.5") -> (Pose) {
    alphaInputType=Bool
    rotationMode=BMM_Additive
    scaleMode=BMM_Additive
    translationMode=BMM_Additive
    alphaBoolBlend {blendInTime=0.1, blendOption=Linear, blendOutTime=0.1}
    boneToModify {boneName=dyn_jacket_A_01}
    rotation {pitch=20, roll=-35}
    scale=<0.5 0.5 0.5>
   }
   n14 = "Transform (Modify) Bone - Bone: dyn_jacket_A_02"!(ComponentPose=@n13, bAlphaBoolEnabled:bool=@n10, Translation="0,0,0", Rotation="0.000,0.000,-25", Scale="1,1,1") -> (Pose) {
    alphaInputType=Bool
    rotationMode=BMM_Additive
    scaleMode=BMM_Additive
    translationMode=BMM_Additive
    alphaBoolBlend {blendInTime=0.1, blendOption=Linear, blendOutTime=0.1}
    boneToModify {boneName=dyn_jacket_A_02}
    rotation {roll=-25}
   }
   n15 = "Transform (Modify) Bone - Bone: dyn_jacket_A_03"!(ComponentPose=@n14, bAlphaBoolEnabled:bool=@n10, Translation="0,0,0", Rotation="0.000,0.000,-20", Scale="1,1,1") -> (Pose) {
    alphaInputType=Bool
    rotationMode=BMM_Additive
    scaleMode=BMM_Additive
    translationMode=BMM_Additive
    alphaBoolBlend {blendInTime=0.1, blendOption=Linear, blendOutTime=0.1}
    boneToModify {boneName=dyn_jacket_A_03}
    rotation {roll=-20}
   }
   n16 = Get Crouching()
   n17 = "Constraint - dyn_r_Strap_MSTR (PPP)"!(
    ComponentPose=@n15
    bAlphaBoolEnabled:bool=@n16
    ConstraintWeights_0=0.0156
    ConstraintWeights_1:float=0.1764
    ConstraintWeights_2:float=0.808
   ) -> (Pose) {
    alphaInputType=Bool
    alphaBoolBlend {blendOption=Linear}
    boneToModify {boneName=dyn_r_Strap_MSTR}
    constraintSetup [{targetBone {boneName=dyn_jacket_F_03}}, {targetBone {boneName=dyn_jacket_F_02}}, {targetBone {boneName=dyn_jacket_F_01}}]
    constraintWeights [1,1,1]
   }
   n18 = "Constraint - dyn_backStrap_MSTR (PPP)"!(
    ComponentPose=@n17
    bAlphaBoolEnabled:bool=@n16
    ConstraintWeights_0=0.1273
    ConstraintWeights_1:float=0.2191
    ConstraintWeights_2:float=0.6536
   ) -> (Pose) {
    alphaInputType=Bool
    alphaBoolBlend {blendOption=Linear}
    boneToModify {boneName=dyn_backStrap_MSTR}
    constraintSetup [{targetBone {boneName=spine_03}}, {targetBone {boneName=spine_04}}, {targetBone {boneName=spine_05}}]
    constraintWeights [1,1,1]
   }
   n19 = "Constraint - dyn_jacket_A_03 (RP)"!(ComponentPose=@n18, bAlphaBoolEnabled:bool=@n1, ConstraintWeights_0=1, ConstraintWeights_1:float=0) -> (Pose) {
    alphaInputType=Bool
    alphaBoolBlend {blendOption=Linear}
    boneToModify {boneName=dyn_jacket_A_03}
    constraintSetup [{transformType=Rotation, targetBone {boneName=thigh_l}}, {targetBone {boneName=pelvis}}]
    constraintWeights [1,1]
   }
   n20 = "Constraint - dyn_jacket_A_04 (RP)"!(ComponentPose=@n19, bAlphaBoolEnabled:bool=@n1, ConstraintWeights_0=1, ConstraintWeights_1:float=0) -> (Pose) {
    alphaInputType=Bool
    alphaBoolBlend {blendOption=Linear}
    boneToModify {boneName=dyn_jacket_A_04}
    constraintSetup [{transformType=Rotation, targetBone {boneName=thigh_l}}, {targetBone {boneName=pelvis}}]
    constraintWeights [1,1]
   }
   n21 = "Constraint - dyn_jacket_B_01 (RP)"!(ComponentPose=@n20, bAlphaBoolEnabled:bool=@n1, ConstraintWeights_0=1, ConstraintWeights_1:float=0) -> (Pose) {
    alphaInputType=Bool
    alphaBoolBlend {blendOption=Linear}
    boneToModify {boneName=dyn_jacket_B_01}
    constraintSetup [{transformType=Rotation, targetBone {boneName=thigh_l}}, {targetBone {boneName=pelvis}}]
    constraintWeights [1,1]
   }
   n22 = "Constraint - dyn_jacket_B_02 (RP)"!(ComponentPose=@n21, bAlphaBoolEnabled:bool=@n1, ConstraintWeights_0=1, ConstraintWeights_1:float=0) -> (Pose) {
    alphaInputType=Bool
    alphaBoolBlend {blendOption=Linear}
    boneToModify {boneName=dyn_jacket_B_02}
    constraintSetup [{transformType=Rotation, targetBone {boneName=thigh_l}}, {targetBone {boneName=pelvis}}]
    constraintWeights [1,1]
   }
   n23 = "Constraint - dyn_jacket_B_03 (RP)"!(ComponentPose=@n22, bAlphaBoolEnabled:bool=@n1, ConstraintWeights_0=1, ConstraintWeights_1:float=0) -> (Pose) {
    alphaInputType=Bool
    alphaBoolBlend {blendOption=Linear}
    boneToModify {boneName=dyn_jacket_B_03}
    constraintSetup [{transformType=Rotation, targetBone {boneName=thigh_l}}, {targetBone {boneName=pelvis}}]
    constraintWeights [1,1]
   }
   n24 = Get Crouching()
   n25 = "Transform (Modify) Bone - Bone: dyn_jacket_A_MSTR"(ComponentPose=@n23, bAlphaBoolEnabled=@n24, Translation="0,10,1", Rotation="0,0,0", Scale="1,1,1") {
    alphaInputType=Bool
    rotationMode=BMM_Additive
    rotationSpace=BCS_BoneSpace
    scaleMode=BMM_Additive
    scaleSpace=BCS_BoneSpace
    translationMode=BMM_Additive
    translationSpace=BCS_BoneSpace
    alphaBoolBlend {blendOption=Linear}
    boneToModify {boneName=dyn_jacket_A_MSTR}
    translation {y=10, z=1}
   }
   n26 = "Transform (Modify) Bone - Bone: dyn_jacket_B_MSTR"(ComponentPose=@n25, bAlphaBoolEnabled=@n24, Translation="3,0,0", Rotation="0,0,0", Scale="1,1,1") {
    alphaInputType=Bool
    rotationMode=BMM_Additive
    scaleMode=BMM_Additive
    translationMode=BMM_Additive
    alphaBoolBlend {blendOption=Linear}
    boneToModify {boneName=dyn_jacket_B_MSTR}
    translation {x=3}
   }
   n27 = "Transform (Modify) Bone - Bone: dyn_l_pouch_MSTR"!(ComponentPose=@n26, bAlphaBoolEnabled:bool=@n24, Translation="0,0,4", Rotation="0,0,10", Scale="1,1,1") -> (Pose) {
    alphaInputType=Bool
    rotationMode=BMM_Additive
    scaleMode=BMM_Additive
    translationMode=BMM_Additive
    alphaBoolBlend {blendOption=Linear}
    boneToModify {boneName=dyn_l_pouch_MSTR}
    rotation {roll=10}
    translation {z=4}
   }
   n28 = "Constraint - dyn_r_Strap_MSTR (PPP)"!(ComponentPose=@n27, Alpha:float=1, ConstraintWeights_0=0.0156, ConstraintWeights_1=0.1764, ConstraintWeights_2=0.808) -> (Pose) {
    alphaBoolBlend {blendOption=Linear}
    boneToModify {boneName=dyn_r_Strap_MSTR}
    constraintSetup [{targetBone {boneName=dyn_jacket_F_03}}, {targetBone {boneName=dyn_jacket_F_02}}, {targetBone {boneName=dyn_jacket_F_01}}]
    constraintWeights [1,1,1]
   }
   n29 = "Constraint - dyn_backStrap_MSTR (PPP)"!(ComponentPose=@n28, Alpha:float=1, ConstraintWeights_0=0.1273, ConstraintWeights_1=0.2191, ConstraintWeights_2=0.6536) -> (Pose) {
    alphaBoolBlend {blendOption=Linear}
    boneToModify {boneName=dyn_backStrap_MSTR}
    constraintSetup [{targetBone {boneName=spine_03}}, {targetBone {boneName=spine_04}}, {targetBone {boneName=spine_05}}]
    constraintWeights [1,1,1]
   }
   n30 = "Constraint - dyn_l_shoeTab_MSTR (PPP)"!(ComponentPose=@n29, Alpha=1, ConstraintWeights_0=0.5707, ConstraintWeights_1:float=0.4217, ConstraintWeights_2:float=0.0077) -> (Pose) {
    alphaBoolBlend {blendOption=Linear}
    boneToModify {boneName=dyn_l_shoeTab_MSTR}
    constraintSetup [{targetBone {boneName=calf_twist_01_l}}, {targetBone {boneName=foot_l}}, {targetBone {boneName=dyn_l_shoeTab}}]
    constraintWeights [1,1,1]
   }
   n31 = "Constraint - dyn_r_shoeTab_MSTR (PPP)"!(ComponentPose=@n30, Alpha=1, ConstraintWeights_0=0.0104, ConstraintWeights_1:float=0.5691, ConstraintWeights_2:float=0.4205) -> (Pose) {
    alphaBoolBlend {blendOption=Linear}
    boneToModify {boneName=dyn_r_shoeTab_MSTR}
    constraintSetup [{targetBone {boneName=dyn_r_shoeTab}}, {targetBone {boneName=calf_twist_01_r}}, {targetBone {boneName=foot_r}}]
    constraintWeights [1,1,1]
   }
   n32 = RigidBody(ComponentPose=@n31) {
    bClampLinearTranslationLimitToRefPose=true
    bOverrideWorldGravity=true
    overridePhysicsAsset=&/M_MED_Snowboarder_Necklace_Physics
    simulationSpace=BaseBoneSpace
    alphaBoolBlend {blendOption=Linear}
    baseBoneRef {boneName=spine_05}
    componentAppliedLinearAccClamp=<5000 5000 5000>
    componentLinearAccScale=<0.5 0.5 0.5>
    overrideWorldGravity {z=-980}
   }
   n33 = Get Gravity_Override_FE() -> (Gravity_Override_FE:Vector)
   n34 = RigidBody!(ComponentPose=@n32, Alpha, OverrideWorldGravity:Vector=@n33, ExternalForce, bTransferBoneVelocities:bool=False) -> (Pose) {
    bClampLinearTranslationLimitToRefPose=true
    bOverrideWorldGravity=true
    overridePhysicsAsset=&/M_MED_Snowboarder_tassel_Physics
    simulationSpace=BaseBoneSpace
    alphaBoolBlend {blendOption=Linear}
    baseBoneRef {boneName=root}
    componentAppliedLinearAccClamp=<1500 1500 1500>
    componentLinearAccScale=<0.4 0.4 0.45>
    overrideWorldGravity {z=-980}
   }
   n35 = RigidBody(ComponentPose=@n34, Alpha=0.2, OverrideWorldGravity=@n33, bTransferBoneVelocities=False) {
    bClampLinearTranslationLimitToRefPose=true
    bOverrideWorldGravity=true
    overridePhysicsAsset=&/M_MED_Snowboarder_shoeTab_Physics
    simulationSpace=BaseBoneSpace
    alphaBoolBlend {blendOption=Linear}
    baseBoneRef {boneName=root}
    componentAppliedLinearAccClamp=<1500 8000 1500>
    overrideWorldGravity {z=-980}
   }
   n36 = Get Skydiving() -> (Skydiving:bool)
   n37 = "Transform (Modify) Bone - Bone: dyn_jacket_C_MSTR"!(ComponentPose=@n35, bAlphaBoolEnabled:bool=@n36, Translation="0,0,0", Rotation="15,0,0", Scale="1,1,1") -> (Pose) {
    alphaInputType=Bool
    rotationMode=BMM_Additive
    rotationSpace=BCS_BoneSpace
    scaleMode=BMM_Additive
    scaleSpace=BCS_BoneSpace
    translationMode=BMM_Additive
    translationSpace=BCS_BoneSpace
    alphaBoolBlend {blendInTime=0.1, blendOption=Linear, blendOutTime=0.1}
    boneToModify {boneName=dyn_jacket_C_MSTR}
    rotation {pitch=15}
   }
   n38 = "Transform (Modify) Bone - Bone: dyn_jacket_D_MSTR"!(ComponentPose=@n37, bAlphaBoolEnabled:bool=@n36, Translation="0,0,0", Rotation="0.000,0.000,10", Scale="1,1,1") -> (Pose) {
    alphaInputType=Bool
    rotationMode=BMM_Additive
    scaleMode=BMM_Additive
    translationMode=BMM_Additive
    alphaBoolBlend {blendInTime=0.1, blendOption=Linear, blendOutTime=0.1}
    boneToModify {boneName=dyn_jacket_D_MSTR}
    rotation {roll=10}
   }
   n39 = "Transform (Modify) Bone - Bone: dyn_jacket_E_MSTR"!(ComponentPose=@n38, bAlphaBoolEnabled:bool=@n36, Translation="0,0,0", Rotation="15,0.000,0.000", Scale="1,1,1") -> (Pose) {
    alphaInputType=Bool
    rotationMode=BMM_Additive
    rotationSpace=BCS_BoneSpace
    scaleMode=BMM_Additive
    scaleSpace=BCS_BoneSpace
    translationMode=BMM_Additive
    translationSpace=BCS_BoneSpace
    alphaBoolBlend {blendInTime=0.1, blendOption=Linear, blendOutTime=0.1}
    boneToModify {boneName=dyn_jacket_E_MSTR}
    rotation {pitch=15}
   }
   n40 = "Transform (Modify) Bone - Bone: dyn_knee"!(
    ComponentPose=@n39
    bAlphaBoolEnabled:bool=Get JumpingFalling() -> (JumpingFalling:bool)
    Translation="0,0,0"
    Rotation="0,0,0"
    Scale="1,1,0.2"
   ) -> (Pose) {
    alphaInputType=Bool
    rotationMode=BMM_Additive
    rotationSpace=BCS_BoneSpace
    scaleMode=BMM_Additive
    scaleSpace=BCS_BoneSpace
    translationMode=BMM_Additive
    translationSpace=BCS_BoneSpace
    alphaBoolBlend {blendInTime=0.1, blendOption=Linear, blendOutTime=0.1}
    boneToModify {boneName=dyn_knee}
    scale {z=0.2}
   }
   n41 = RigidBody!(ComponentPose=@n40, Alpha=0.5, ExternalForce, bTransferBoneVelocities=False) -> (Pose) {
    bClampLinearTranslationLimitToRefPose=true
    bOverrideWorldGravity=true
    overridePhysicsAsset=&/M_MED_Snowboarder_Strap_Physics
    simulationSpace=BaseBoneSpace
    alphaBoolBlend {blendOption=Linear}
    baseBoneRef {boneName=root}
    componentLinearAccScale=<0.25 0.2 0.35>
    overrideWorldGravity {z=-980}
   }
   n42 = SelectFloat(A=1, B=0, bPickA=Get bIsInFrontEnd())
   n43 = "Transform (Modify) Bone - Bone: dyn_jacket_G_MSTR"(ComponentPose=@n41, Alpha=@n42, Translation="0,0,0", Rotation="10,0,0", Scale="1,1,1") {
    rotationMode=BMM_Additive
    rotationSpace=BCS_BoneSpace
    scaleMode=BMM_Additive
    scaleSpace=BCS_BoneSpace
    translationMode=BMM_Additive
    translationSpace=BCS_BoneSpace
    alphaBoolBlend {blendOption=Linear}
    boneToModify {boneName=dyn_jacket_G_MSTR}
    rotation {pitch=10}
   }
   n44 = "Transform (Modify) Bone - Bone: dyn_jacket_A_MSTR"!(ComponentPose=@n43, Alpha:float=@n42, Translation="0,0,0", Rotation="8,0,0", Scale="1,1,1") -> (Pose) {
    rotationMode=BMM_Additive
    rotationSpace=BCS_BoneSpace
    scaleMode=BMM_Additive
    scaleSpace=BCS_BoneSpace
    translationMode=BMM_Additive
    translationSpace=BCS_BoneSpace
    alphaBoolBlend {blendOption=Linear}
    boneToModify {boneName=dyn_jacket_A_MSTR}
    rotation {pitch=8}
   }
   n45 = Anim Dynamics - Bone: dyn_jacket_A_01!(ComponentPose=@n44, Alpha=0.3, GravityOverride:Vector="0,0,0", ExternalForce="0,0,0") -> (Pose) {
    angularDampingOverride=0.9999
    angularSpringConstant=75
    bAngularSpring=true
    bGravityOverrideInSimSpace=true
    bOverrideAngularDamping=true
    simulationSpace=RootRelative
    alphaBoolBlend {blendOption=Linear}
    boundBone {boneName=dyn_jacket_A_01}
    componentAppliedLinearAccClamp=<2500 2500 2500>
    componentLinearAccScale=<1 1 1>
    componentLinearVelScale=<1 1 1>
    physicsBodyDefinitions [{
      sphereCollisionRadius=0
      boundBone {boneName=dyn_jacket_A_01}
      boxExtents=<10 20 20>
      constraintSetup {angularLimitsMax=<25 0 25>, angularLimitsMin=<-25 -25 -25>, angularTarget=<0 0 0>}
      localJointOffset=<2 0 0>
     }]
   }
   n46 = Anim Dynamics - Bone: dyn_jacket_A_02!(ComponentPose=@n45, Alpha=0.3, GravityOverride:Vector="0,0,0", ExternalForce="0,0,0") -> (Pose) {
    angularDampingOverride=0.9999
    angularSpringConstant=75
    bAngularSpring=true
    bGravityOverrideInSimSpace=true
    bOverrideAngularDamping=true
    simulationSpace=RootRelative
    alphaBoolBlend {blendOption=Linear}
    boundBone {boneName=dyn_jacket_A_02}
    componentAppliedLinearAccClamp=<2500 2500 2500>
    componentLinearAccScale=<1 1 1>
    componentLinearVelScale=<1 1 1>
    physicsBodyDefinitions [{
      sphereCollisionRadius=0
      boundBone {boneName=dyn_jacket_A_02}
      boxExtents=<10 20 20>
      constraintSetup {angularLimitsMax=<25 0 25>, angularLimitsMin=<-25 -25 -25>, angularTarget=<0 0 0>}
      localJointOffset=<2 0 0>
     }]
   }
   n47 = Anim Dynamics - Bone: dyn_jacket_A_03!(ComponentPose=@n46, Alpha=1, GravityOverride:Vector="0,0,0", ExternalForce="0,0,0") -> (Pose) {
    angularDampingOverride=0.9999
    angularSpringConstant=75
    bAngularSpring=true
    bGravityOverrideInSimSpace=true
    bOverrideAngularDamping=true
    simulationSpace=RootRelative
    alphaBoolBlend {blendOption=Linear}
    boundBone {boneName=dyn_jacket_A_03}
    componentAppliedLinearAccClamp=<2500 2500 2500>
    componentLinearAccScale=<1 1 1>
    componentLinearVelScale=<1 1 1>
    physicsBodyDefinitions [{
      sphereCollisionRadius=0
      boundBone {boneName=dyn_jacket_A_03}
      boxExtents=<10 20 20>
      constraintSetup {angularLimitsMax=<25 0 25>, angularLimitsMin=<-25 -25 -25>, angularTarget=<0 0 0>}
      localJointOffset=<2 0 0>
     }]
   }
   n48 = Anim Dynamics - Bone: dyn_jacket_A_04!(ComponentPose=@n47, Alpha=1, GravityOverride:Vector="0,0,0", ExternalForce="0,0,0") -> (Pose) {
    angularDampingOverride=0.9999
    angularSpringConstant=75
    bAngularSpring=true
    bGravityOverrideInSimSpace=true
    bOverrideAngularDamping=true
    simulationSpace=RootRelative
    alphaBoolBlend {blendOption=Linear}
    boundBone {boneName=dyn_jacket_A_04}
    componentAppliedLinearAccClamp=<2500 2500 2500>
    componentLinearAccScale=<1 1 1>
    componentLinearVelScale=<1 1 1>
    physicsBodyDefinitions [{
      sphereCollisionRadius=0
      boundBone {boneName=dyn_jacket_A_04}
      boxExtents=<10 20 20>
      constraintSetup {angularLimitsMax=<25 0 25>, angularLimitsMin=<-25 -25 -25>, angularTarget=<0 0 0>}
      localJointOffset=<2 0 0>
     }]
   }
   n49 = Anim Dynamics - Bone: dyn_jacket_B_01!(ComponentPose=@n48, Alpha=1, GravityOverride:Vector="0,0,0", ExternalForce="0,0,0") -> (Pose) {
    angularDampingOverride=0.9999
    angularSpringConstant=75
    bAngularSpring=true
    bGravityOverrideInSimSpace=true
    bOverrideAngularDamping=true
    simulationSpace=RootRelative
    alphaBoolBlend {blendOption=Linear}
    boundBone {boneName=dyn_jacket_B_01}
    componentAppliedLinearAccClamp=<2500 2500 2500>
    componentLinearAccScale=<1 1 1>
    componentLinearVelScale=<1 1 1>
    physicsBodyDefinitions [{
      sphereCollisionRadius=0
      boundBone {boneName=dyn_jacket_B_01}
      boxExtents=<10 20 20>
      constraintSetup {angularLimitsMax=<25 -3 25>, angularLimitsMin=<-25 -10 -25>, angularTarget=<0 0 0>}
      localJointOffset=<2 0 0>
     }]
   }
   n50 = Anim Dynamics - Bone: dyn_jacket_B_02!(ComponentPose=@n49, Alpha=1, GravityOverride:Vector="0,0,0", ExternalForce="0,0,0") -> (Pose) {
    angularDampingOverride=0.9999
    angularSpringConstant=75
    bAngularSpring=true
    bGravityOverrideInSimSpace=true
    bOverrideAngularDamping=true
    simulationSpace=RootRelative
    alphaBoolBlend {blendOption=Linear}
    boundBone {boneName=dyn_jacket_B_02}
    componentAppliedLinearAccClamp=<2500 2500 2500>
    componentLinearAccScale=<1 1 1>
    componentLinearVelScale=<1 1 1>
    physicsBodyDefinitions [{
      sphereCollisionRadius=0
      boundBone {boneName=dyn_jacket_B_02}
      boxExtents=<10 20 20>
      constraintSetup {angularLimitsMax=<25 -3 25>, angularLimitsMin=<-25 -10 -25>, angularTarget=<0 0 0>}
      localJointOffset=<2 0 0>
     }]
   }
   n51 = Anim Dynamics - Bone: dyn_jacket_B_03!(ComponentPose=@n50, Alpha=1, GravityOverride:Vector="0,0,0", ExternalForce="0,0,0") -> (Pose) {
    angularDampingOverride=0.9999
    angularSpringConstant=75
    bAngularSpring=true
    bGravityOverrideInSimSpace=true
    bOverrideAngularDamping=true
    simulationSpace=RootRelative
    alphaBoolBlend {blendOption=Linear}
    boundBone {boneName=dyn_jacket_B_03}
    componentAppliedLinearAccClamp=<2500 2500 2500>
    componentLinearAccScale=<1 1 1>
    componentLinearVelScale=<1 1 1>
    physicsBodyDefinitions [{
      sphereCollisionRadius=0
      boundBone {boneName=dyn_jacket_B_03}
      boxExtents=<10 20 20>
      constraintSetup {angularLimitsMax=<25 -3 25>, angularLimitsMin=<-25 -10 -25>, angularTarget=<0 0 0>}
      localJointOffset=<2 0 0>
     }]
   }
   n52 = Anim Dynamics - Bone: dyn_jacket_C_01!(ComponentPose=@n51, Alpha=1, GravityOverride:Vector="0,0,0", ExternalForce="0,0,0") -> (Pose) {
    angularDampingOverride=0.9999
    angularSpringConstant=75
    bAngularSpring=true
    bGravityOverrideInSimSpace=true
    bOverrideAngularDamping=true
    simulationSpace=RootRelative
    alphaBoolBlend {blendOption=Linear}
    boundBone {boneName=dyn_jacket_C_01}
    componentAppliedLinearAccClamp=<2500 2500 2500>
    componentLinearAccScale=<1 1 1>
    componentLinearVelScale=<1 1 1>
    physicsBodyDefinitions [{
      sphereCollisionRadius=0
      boundBone {boneName=dyn_jacket_C_01}
      boxExtents=<10 20 20>
      constraintSetup {angularLimitsMax=<25 -3 25>, angularLimitsMin=<-25 -10 -25>, angularTarget=<0 0 0>}
      localJointOffset=<2 0 0>
     }]
   }
   n53 = Anim Dynamics - Bone: dyn_jacket_C_02!(ComponentPose=@n52, Alpha=1, GravityOverride:Vector="0,0,0", ExternalForce="0,0,0") -> (Pose) {
    angularDampingOverride=0.9999
    angularSpringConstant=75
    bAngularSpring=true
    bGravityOverrideInSimSpace=true
    bOverrideAngularDamping=true
    simulationSpace=RootRelative
    alphaBoolBlend {blendOption=Linear}
    boundBone {boneName=dyn_jacket_C_02}
    componentAppliedLinearAccClamp=<2500 2500 2500>
    componentLinearAccScale=<1 1 1>
    componentLinearVelScale=<1 1 1>
    physicsBodyDefinitions [{
      sphereCollisionRadius=0
      boundBone {boneName=dyn_jacket_C_02}
      boxExtents=<10 20 20>
      constraintSetup {angularLimitsMax=<25 -3 25>, angularLimitsMin=<-25 -10 -25>, angularTarget=<0 0 0>}
      localJointOffset=<2 0 0>
     }]
   }
   n54 = Anim Dynamics - Bone: dyn_jacket_C_03!(ComponentPose=@n53, Alpha=1, GravityOverride:Vector="0,0,0", ExternalForce="0,0,0") -> (Pose) {
    angularDampingOverride=0.9999
    angularSpringConstant=75
    bAngularSpring=true
    bGravityOverrideInSimSpace=true
    bOverrideAngularDamping=true
    simulationSpace=RootRelative
    alphaBoolBlend {blendOption=Linear}
    boundBone {boneName=dyn_jacket_C_03}
    componentAppliedLinearAccClamp=<2500 2500 2500>
    componentLinearAccScale=<1 1 1>
    componentLinearVelScale=<1 1 1>
    physicsBodyDefinitions [{
      sphereCollisionRadius=0
      boundBone {boneName=dyn_jacket_C_03}
      boxExtents=<10 20 20>
      constraintSetup {angularLimitsMax=<25 -3 25>, angularLimitsMin=<-25 -10 -25>, angularTarget=<0 0 0>}
      localJointOffset=<2 0 0>
     }]
   }
   n55 = Anim Dynamics - Bone: dyn_jacket_D_01!(ComponentPose=@n54, Alpha=1, GravityOverride:Vector="0,0,0", ExternalForce="0,0,0") -> (Pose) {
    angularDampingOverride=0.9999
    angularSpringConstant=75
    bAngularSpring=true
    bGravityOverrideInSimSpace=true
    bOverrideAngularDamping=true
    simulationSpace=RootRelative
    alphaBoolBlend {blendOption=Linear}
    boundBone {boneName=dyn_jacket_D_01}
    componentAppliedLinearAccClamp=<2500 2500 2500>
    componentLinearAccScale=<1 1 1>
    componentLinearVelScale=<1 1 1>
    physicsBodyDefinitions [{
      sphereCollisionRadius=0
      boundBone {boneName=dyn_jacket_D_01}
      boxExtents=<10 20 20>
      constraintSetup {angularLimitsMax=<25 -3 25>, angularLimitsMin=<-25 -20 -25>, angularTarget=<0 0 0>}
      localJointOffset=<2 0 0>
     }]
   }
   n56 = Anim Dynamics - Bone: dyn_jacket_D_02!(ComponentPose=@n55, Alpha=1, GravityOverride:Vector="0,0,0", ExternalForce="0,0,0") -> (Pose) {
    angularDampingOverride=0.9999
    angularSpringConstant=75
    bAngularSpring=true
    bGravityOverrideInSimSpace=true
    bOverrideAngularDamping=true
    simulationSpace=RootRelative
    alphaBoolBlend {blendOption=Linear}
    boundBone {boneName=dyn_jacket_D_02}
    componentAppliedLinearAccClamp=<2500 2500 2500>
    componentLinearAccScale=<1 1 1>
    componentLinearVelScale=<1 1 1>
    physicsBodyDefinitions [{
      sphereCollisionRadius=0
      boundBone {boneName=dyn_jacket_D_02}
      boxExtents=<10 20 20>
      constraintSetup {angularLimitsMax=<25 -3 25>, angularLimitsMin=<-25 -20 -25>, angularTarget=<0 0 0>}
      localJointOffset=<2 0 0>
     }]
   }
   n57 = Anim Dynamics - Bone: dyn_jacket_D_03!(ComponentPose=@n56, Alpha=1, GravityOverride:Vector="0,0,0", ExternalForce="0,0,0") -> (Pose) {
    angularDampingOverride=0.9999
    angularSpringConstant=75
    bAngularSpring=true
    bGravityOverrideInSimSpace=true
    bOverrideAngularDamping=true
    simulationSpace=RootRelative
    alphaBoolBlend {blendOption=Linear}
    boundBone {boneName=dyn_jacket_D_03}
    componentAppliedLinearAccClamp=<2500 2500 2500>
    componentLinearAccScale=<1 1 1>
    componentLinearVelScale=<1 1 1>
    physicsBodyDefinitions [{
      sphereCollisionRadius=0
      boundBone {boneName=dyn_jacket_D_03}
      boxExtents=<10 20 20>
      constraintSetup {angularLimitsMax=<25 -3 25>, angularLimitsMin=<-25 -20 -25>, angularTarget=<0 0 0>}
      localJointOffset=<2 0 0>
     }]
   }
   n58 = Anim Dynamics - Bone: dyn_jacket_E_01!(ComponentPose=@n57, Alpha=1, GravityOverride:Vector="0,0,0", ExternalForce="0,0,0") -> (Pose) {
    angularDampingOverride=0.9999
    angularSpringConstant=75
    bAngularSpring=true
    bGravityOverrideInSimSpace=true
    bOverrideAngularDamping=true
    simulationSpace=RootRelative
    alphaBoolBlend {blendOption=Linear}
    boundBone {boneName=dyn_jacket_E_01}
    componentAppliedLinearAccClamp=<2500 2500 2500>
    componentLinearAccScale=<1 1 1>
    componentLinearVelScale=<1 1 1>
    physicsBodyDefinitions [{
      sphereCollisionRadius=0
      boundBone {boneName=dyn_jacket_E_01}
      boxExtents=<10 20 20>
      constraintSetup {angularLimitsMax=<25 -3 25>, angularLimitsMin=<-25 -10 -25>, angularTarget=<0 0 0>}
      localJointOffset=<2 0 0>
     }]
   }
   n59 = Anim Dynamics - Bone: dyn_jacket_E_02!(ComponentPose=@n58, Alpha=1, GravityOverride:Vector="0,0,0", ExternalForce="0,0,0") -> (Pose) {
    angularDampingOverride=0.9999
    angularSpringConstant=75
    bAngularSpring=true
    bGravityOverrideInSimSpace=true
    bOverrideAngularDamping=true
    simulationSpace=RootRelative
    alphaBoolBlend {blendOption=Linear}
    boundBone {boneName=dyn_jacket_E_02}
    componentAppliedLinearAccClamp=<2500 2500 2500>
    componentLinearAccScale=<1 1 1>
    componentLinearVelScale=<1 1 1>
    physicsBodyDefinitions [{
      sphereCollisionRadius=0
      boundBone {boneName=dyn_jacket_E_02}
      boxExtents=<10 20 20>
      constraintSetup {angularLimitsMax=<25 -3 25>, angularLimitsMin=<-25 -10 -25>, angularTarget=<0 0 0>}
      localJointOffset=<2 0 0>
     }]
   }
   n60 = Anim Dynamics - Bone: dyn_jacket_E_03!(ComponentPose=@n59, Alpha=1, GravityOverride:Vector="0,0,0", ExternalForce="0,0,0") -> (Pose) {
    angularDampingOverride=0.9999
    angularSpringConstant=75
    bAngularSpring=true
    bGravityOverrideInSimSpace=true
    bOverrideAngularDamping=true
    simulationSpace=RootRelative
    alphaBoolBlend {blendOption=Linear}
    boundBone {boneName=dyn_jacket_E_03}
    componentAppliedLinearAccClamp=<2500 2500 2500>
    componentLinearAccScale=<1 1 1>
    componentLinearVelScale=<1 1 1>
    physicsBodyDefinitions [{
      sphereCollisionRadius=0
      boundBone {boneName=dyn_jacket_E_03}
      boxExtents=<10 20 20>
      constraintSetup {angularLimitsMax=<25 -3 25>, angularLimitsMin=<-25 -10 -25>, angularTarget=<0 0 0>}
      localJointOffset=<2 0 0>
     }]
   }
   n61 = Anim Dynamics - Bone: dyn_jacket_F_01!(ComponentPose=@n60, Alpha=1, GravityOverride:Vector="0,0,0", ExternalForce="0,0,0") -> (Pose) {
    angularDampingOverride=0.9999
    angularSpringConstant=75
    bAngularSpring=true
    bGravityOverrideInSimSpace=true
    bOverrideAngularDamping=true
    simulationSpace=RootRelative
    alphaBoolBlend {blendOption=Linear}
    boundBone {boneName=dyn_jacket_F_01}
    componentAppliedLinearAccClamp=<2500 2500 2500>
    componentLinearAccScale=<1 1 1>
    componentLinearVelScale=<1 1 1>
    physicsBodyDefinitions [{
      sphereCollisionRadius=0
      boundBone {boneName=dyn_jacket_F_01}
      boxExtents=<10 20 20>
      constraintSetup {angularLimitsMax=<25 -3 25>, angularLimitsMin=<-25 -10 -25>, angularTarget=<0 0 0>}
      localJointOffset=<2 0 0>
     }]
   }
   n62 = Anim Dynamics - Bone: dyn_jacket_F_02!(ComponentPose=@n61, Alpha=1, GravityOverride:Vector="0,0,0", ExternalForce="0,0,0") -> (Pose) {
    angularDampingOverride=0.9999
    angularSpringConstant=75
    bAngularSpring=true
    bGravityOverrideInSimSpace=true
    bOverrideAngularDamping=true
    simulationSpace=RootRelative
    alphaBoolBlend {blendOption=Linear}
    boundBone {boneName=dyn_jacket_F_02}
    componentAppliedLinearAccClamp=<2500 2500 2500>
    componentLinearAccScale=<1 1 1>
    componentLinearVelScale=<1 1 1>
    physicsBodyDefinitions [{
      sphereCollisionRadius=0
      boundBone {boneName=dyn_jacket_F_02}
      boxExtents=<10 20 20>
      constraintSetup {angularLimitsMax=<25 -3 25>, angularLimitsMin=<-25 -10 -25>, angularTarget=<0 0 0>}
      localJointOffset=<2 0 0>
     }]
   }
   n63 = Anim Dynamics - Bone: dyn_jacket_F_03!(ComponentPose=@n62, Alpha=1, GravityOverride:Vector="0,0,0", ExternalForce="0,0,0") -> (Pose) {
    angularDampingOverride=0.9999
    angularSpringConstant=75
    bAngularSpring=true
    bGravityOverrideInSimSpace=true
    bOverrideAngularDamping=true
    simulationSpace=RootRelative
    alphaBoolBlend {blendOption=Linear}
    boundBone {boneName=dyn_jacket_F_03}
    componentAppliedLinearAccClamp=<2500 2500 2500>
    componentLinearAccScale=<1 1 1>
    componentLinearVelScale=<1 1 1>
    physicsBodyDefinitions [{
      sphereCollisionRadius=0
      boundBone {boneName=dyn_jacket_F_03}
      boxExtents=<10 20 20>
      constraintSetup {angularLimitsMax=<25 -3 25>, angularLimitsMin=<-25 -10 -25>, angularTarget=<0 0 0>}
      localJointOffset=<2 0 0>
     }]
   }
   n64 = Anim Dynamics - Bone: dyn_jacket_G_01!(ComponentPose=@n63, Alpha=1, GravityOverride:Vector="0,0,0", ExternalForce="0,0,0") -> (Pose) {
    angularDampingOverride=0.9999
    angularSpringConstant=75
    bAngularSpring=true
    bGravityOverrideInSimSpace=true
    bOverrideAngularDamping=true
    simulationSpace=RootRelative
    alphaBoolBlend {blendOption=Linear}
    boundBone {boneName=dyn_jacket_G_01}
    componentAppliedLinearAccClamp=<2500 2500 2500>
    componentLinearAccScale=<1 1 1>
    componentLinearVelScale=<1 1 1>
    physicsBodyDefinitions [{
      sphereCollisionRadius=0
      boundBone {boneName=dyn_jacket_G_01}
      boxExtents=<10 20 20>
      constraintSetup {angularLimitsMax=<25 0 25>, angularLimitsMin=<-25 -25 -25>, angularTarget=<0 0 0>}
      localJointOffset=<2 0 0>
     }]
   }
   n65 = Anim Dynamics - Bone: dyn_jacket_G_02!(ComponentPose=@n64, Alpha=1, GravityOverride:Vector="0,0,0", ExternalForce="0,0,0") -> (Pose) {
    angularDampingOverride=0.9999
    angularSpringConstant=75
    bAngularSpring=true
    bGravityOverrideInSimSpace=true
    bOverrideAngularDamping=true
    simulationSpace=RootRelative
    alphaBoolBlend {blendOption=Linear}
    boundBone {boneName=dyn_jacket_G_02}
    componentAppliedLinearAccClamp=<2500 2500 2500>
    componentLinearAccScale=<1 1 1>
    componentLinearVelScale=<1 1 1>
    physicsBodyDefinitions [{
      sphereCollisionRadius=0
      boundBone {boneName=dyn_jacket_G_02}
      boxExtents=<10 20 20>
      constraintSetup {angularLimitsMax=<25 0 25>, angularLimitsMin=<-25 -25 -25>, angularTarget=<0 0 0>}
      localJointOffset=<2 0 0>
     }]
   }
   n66 = Anim Dynamics - Bone: dyn_jacket_G_03!(ComponentPose=@n65, Alpha=1, GravityOverride:Vector="0,0,0", ExternalForce="0,0,0") -> (Pose) {
    angularDampingOverride=0.9999
    angularSpringConstant=75
    bAngularSpring=true
    bGravityOverrideInSimSpace=true
    bOverrideAngularDamping=true
    simulationSpace=RootRelative
    alphaBoolBlend {blendOption=Linear}
    boundBone {boneName=dyn_jacket_G_03}
    componentAppliedLinearAccClamp=<2500 2500 2500>
    componentLinearAccScale=<1 1 1>
    componentLinearVelScale=<1 1 1>
    physicsBodyDefinitions [{
      sphereCollisionRadius=0
      boundBone {boneName=dyn_jacket_G_03}
      boxExtents=<10 20 20>
      constraintSetup {angularLimitsMax=<25 0 25>, angularLimitsMin=<-25 -25 -25>, angularTarget=<0 0 0>}
      localJointOffset=<2 0 0>
     }]
   }
   n67 = Anim Dynamics - Bone: dyn_jacket_G_04!(ComponentPose=@n66, Alpha=1, GravityOverride:Vector="0,0,0", ExternalForce="0,0,0") -> (Pose) {
    angularDampingOverride=0.9999
    angularSpringConstant=75
    bAngularSpring=true
    bGravityOverrideInSimSpace=true
    bOverrideAngularDamping=true
    simulationSpace=RootRelative
    alphaBoolBlend {blendOption=Linear}
    boundBone {boneName=dyn_jacket_G_04}
    componentAppliedLinearAccClamp=<2500 2500 2500>
    componentLinearAccScale=<1 1 1>
    componentLinearVelScale=<1 1 1>
    physicsBodyDefinitions [{
      sphereCollisionRadius=0
      boundBone {boneName=dyn_jacket_G_04}
      boxExtents=<10 20 20>
      constraintSetup {angularLimitsMax=<25 0 25>, angularLimitsMin=<-25 -25 -25>, angularTarget=<0 0 0>}
      localJointOffset=<2 0 0>
     }]
   }
   n68 = RigidBody!(ComponentPose=@n67, Alpha, OverrideWorldGravity:Vector=Get Gravity_Override_FE(), ExternalForce) -> (Pose) {
    bClampLinearTranslationLimitToRefPose=true
    bOverrideWorldGravity=true
    overridePhysicsAsset=&/M_MED_Snowboarder_Jacket_Physics
    simulationSpace=BaseBoneSpace
    alphaBoolBlend {blendOption=Linear}
    baseBoneRef {boneName=root}
    componentAppliedLinearAccClamp=<1000 1000 1000>
    componentLinearAccScale=<0.2 0.2 0.2>
    overrideWorldGravity {z=-980}
   }
   n69 = "Transform (Modify) Bone - Bone: dyn_jacket_C_MSTR"!(ComponentPose=@n68, Alpha:float=1, Translation="0,0,3", Rotation="0,0,0", Scale="1,1,1") -> (Pose) {
    rotationMode=BMM_Additive
    rotationSpace=BCS_BoneSpace
    scaleMode=BMM_Replace
    scaleSpace=BCS_BoneSpace
    translationMode=BMM_Additive
    translationSpace=BCS_BoneSpace
    alphaBoolBlend {blendOption=Linear}
    boneToModify {boneName=dyn_jacket_C_MSTR}
    translation {z=3}
   }
   n70 = "Transform (Modify) Bone - Bone: dyn_jacket_D_MSTR"!(ComponentPose=@n69, Alpha:float=1, Translation="0,0,3", Rotation="0,0,0", Scale="1,1,1") -> (Pose) {
    rotationMode=BMM_Additive
    rotationSpace=BCS_BoneSpace
    scaleMode=BMM_Replace
    scaleSpace=BCS_BoneSpace
    translationMode=BMM_Additive
    translationSpace=BCS_BoneSpace
    alphaBoolBlend {blendOption=Linear}
    boneToModify {boneName=dyn_jacket_D_MSTR}
    translation {z=3}
   }
   n71 = "Transform (Modify) Bone - Bone: dyn_jacket_E_MSTR"!(ComponentPose=@n70, Alpha:float=1, Translation="0,0,3", Rotation="0,0,0", Scale="1,1,1") -> (Pose) {
    rotationMode=BMM_Additive
    rotationSpace=BCS_BoneSpace
    scaleMode=BMM_Replace
    scaleSpace=BCS_BoneSpace
    translationMode=BMM_Additive
    translationSpace=BCS_BoneSpace
    alphaBoolBlend {blendOption=Linear}
    boneToModify {boneName=dyn_jacket_E_MSTR}
    translation {z=3}
   }
   n72 = OR Boolean(A=AND Boolean(A=Get bIsCrouching(), B=Get bIsAccelerating2D()), B=OR Boolean(A=Get bIsJumping(), B=Get bIsFalling()))
   n73 = "Transform (Modify) Bone - Bone: dyn_jacket_G_MSTR"!(ComponentPose=@n71, bAlphaBoolEnabled:bool=@n72, Translation="-3,0,0", Rotation="-30,20,10", Scale="1,1,1") -> (Pose) {
    alphaInputType=Bool
    rotationMode=BMM_Additive
    scaleMode=BMM_Additive
    translationMode=BMM_Additive
    alphaBoolBlend {blendInTime=0.3, blendOption=Linear, blendOutTime=0.3}
    boneToModify {boneName=dyn_jacket_G_MSTR}
    rotation=<rot -30 20 10>
    translation {x=-3}
   }
   n74 = "Transform (Modify) Bone - Bone: dyn_jacket_F_MSTR"!(ComponentPose=@n73, bAlphaBoolEnabled:bool=@n72, Translation="-3,0,0", Rotation="-10,0,0", Scale="1,1,1") -> (Pose) {
    alphaInputType=Bool
    rotationMode=BMM_Additive
    scaleMode=BMM_Additive
    translationMode=BMM_Additive
    alphaBoolBlend {blendInTime=0.3, blendOption=Linear, blendOutTime=0.3}
    boneToModify {boneName=dyn_jacket_F_MSTR}
    rotation {pitch=-10}
    translation {x=-3}
   }
   n75 = Anim Dynamics - Bone: dyn_backStrap(ComponentPose=@n74, Alpha=1, ExternalForce="0,0,0") {
    angularDampingOverride=0.9999
    angularSpringConstant=75
    bAngularSpring=true
    bGravityOverrideInSimSpace=true
    bOverrideAngularDamping=true
    simulationSpace=RootRelative
    alphaBoolBlend {blendOption=Linear}
    boundBone {boneName=dyn_backStrap}
    componentAppliedLinearAccClamp=<2500 2500 3500>
    componentLinearAccScale=<3 3 3>
    componentLinearVelScale=<3 3 3>
    physicsBodyDefinitions [{
      sphereCollisionRadius=0
      boundBone {boneName=dyn_backStrap}
      boxExtents=<10 20 20>
      constraintSetup {angularLimitsMax=<25 5 25>, angularLimitsMin=<-25 -70 -25>, angularTarget=<0 0 0>}
      localJointOffset=<2 0 -0.25>
     }]
   }
   n76 = "Constraint - dyn_jacket_A_04 (PP)"!(
    ComponentPose=Local To Component(LocalPose=Copy Pose From Mesh() {bCopyCurves=true, bUseAttachedParent=true})
    Alpha=1
    ConstraintWeights_0=0.9968
    ConstraintWeights_1:float=0.0032
   ) -> (Pose) {
    alphaBoolBlend {blendOption=Linear}
    boneToModify {boneName=dyn_jacket_A_04}
    constraintSetup [{targetBone {boneName=thigh_twist_01_l}}, {targetBone {boneName=dyn_jacket_B_03}}]
    constraintWeights [1,1]
   }
   n77 = "Transform (Modify) Bone - Bone: dyn_jacket_G_MSTR"(ComponentPose=@n76, bAlphaBoolEnabled=True, Translation="-3,0,0", Rotation="-30,20,10", Scale="1,1,1") {
    alphaInputType=Bool
    rotationMode=BMM_Additive
    scaleMode=BMM_Additive
    translationMode=BMM_Additive
    alphaBoolBlend {blendInTime=0.3, blendOption=Linear, blendOutTime=0.3}
    boneToModify {boneName=dyn_jacket_G_MSTR}
    rotation=<rot -30 20 10>
    translation {x=-3}
   }
   n78 = "Transform (Modify) Bone - Bone: dyn_jacket_F_MSTR"(ComponentPose=@n77, bAlphaBoolEnabled=True, Translation="-3,0,0", Rotation="-10,0,0", Scale="1,1,1") {
    alphaInputType=Bool
    rotationMode=BMM_Additive
    scaleMode=BMM_Additive
    translationMode=BMM_Additive
    alphaBoolBlend {blendInTime=0.3, blendOption=Linear, blendOutTime=0.3}
    boneToModify {boneName=dyn_jacket_F_MSTR}
    rotation {pitch=-10}
    translation {x=-3}
   }
   n79 = RigidBody!(ComponentPose=@n78, Alpha, ExternalForce) -> (Pose) {
    bClampLinearTranslationLimitToRefPose=true
    bOverrideWorldGravity=true
    overridePhysicsAsset=&M_MED_Snowboarder_Necklace_Physics
    simulationSpace=BaseBoneSpace
    alphaBoolBlend {blendOption=Linear}
    baseBoneRef {boneName=spine_05}
    componentAppliedLinearAccClamp=<5000 5000 5000>
    componentLinearAccScale=<0.5 0.5 0.5>
    overrideWorldGravity {z=-980}
   }
   n80 = Get Gravity_Override_FE()
   n81 = RigidBody!(ComponentPose=@n79, Alpha, OverrideWorldGravity:Vector=@n80, ExternalForce, bTransferBoneVelocities:bool=False) -> (Pose) {
    bClampLinearTranslationLimitToRefPose=true
    bOverrideWorldGravity=true
    overridePhysicsAsset=&M_MED_Snowboarder_tassel_Physics
    simulationSpace=BaseBoneSpace
    alphaBoolBlend {blendOption=Linear}
    baseBoneRef {boneName=root}
    componentAppliedLinearAccClamp=<1500 1500 1500>
    componentLinearAccScale=<0.4 0.4 0.45>
    overrideWorldGravity {z=-980}
   }
   n82 = RigidBody(ComponentPose=@n81, Alpha=0.2, OverrideWorldGravity=@n80, bTransferBoneVelocities=False) {
    bClampLinearTranslationLimitToRefPose=true
    bOverrideWorldGravity=true
    overridePhysicsAsset=&/M_MED_Snowboarder_Jacket_Crouch_Physics
    simulationSpace=BaseBoneSpace
    alphaBoolBlend {blendOption=Linear}
    baseBoneRef {boneName=root}
    componentAppliedLinearAccClamp=<1500 8000 1500>
    overrideWorldGravity {z=-980}
   }
   n83 = RigidBody!(ComponentPose=@n82, Alpha, ExternalForce=@n80) -> (Pose) {
    bClampLinearTranslationLimitToRefPose=true
    bOverrideWorldGravity=true
    overridePhysicsAsset=&/M_MED_Snowboarder_Jacket_Emote_Physics
    simulationSpace=BaseBoneSpace
    alphaBoolBlend {blendOption=Linear}
    baseBoneRef {boneName=pelvis}
    componentAppliedLinearAccClamp=<5000 5000 5000>
    componentLinearAccScale=<0.25 0.25 0.4>
    overrideWorldGravity {z=-980}
   }
   Output Pose(Result=Blend Poses by bool(BlendPose_0=Component To Local(ComponentPose=@n83), BlendPose_1=Component To Local(ComponentPose=@n75), bActiveValue=Get bIsPlayingEmote()) {blendType=Linear}) {layerGroup=None}
  ]}
 EventGraph {nodes [
   Event BlueprintUpdateAnimation()
   > n1 = Sequence!(execute) -> (then_0, then_1, then_2:Exec, then_3:Exec, then_4:Exec)
   n2 = GetDeltaTransformFromRefPose(self=GetOwnerSkeletalMeshComponent(), BoneName=upperarm_l, BaseName=clavicle_l)
   n3 = "Absolute (Float)"(A=BreakRotator(InRot=BreakTransform(InTransform=@n2).Rotation).Yaw)
   > Set Arm Side Lift L(Arm Side Lift L:double=float * float(A=MapRangeClamped(Value=@n3, InRangeA=45, InRangeB=20), B=1)) -> (Output_Get:double)
   > n4 = Cast To FortPlayerAnimInstance(Object=GetAnimInstance(self=GetOwnerSkeletalMeshComponent()))
   > Set isAimingPistol(isAimingPistol:double=0) -> (Output_Get:double)
   > Set isAimingRifle(isAimingRifle:bool=false) -> (Output_Get:bool)
   > n5 = Switch on EFortWeaponCoreAnimation!(execute, Selection:EFortWeaponCoreAnimation Enum=Get WeaponCoreAnim(self=@n4)) -> (
    Melee:Exec
    Pistol:Exec
    Shotgun:Exec
    PaperBlueprint:Exec
    Rifle:Exec
    MeleeOneHand:Exec
    MachinePistol:Exec
    RocketLauncher:Exec
    GrenadeLauncher:Exec
    GoingCommando:Exec
    AssaultRifle:Exec
    TacticalShotgun:Exec
    SniperRifle:Exec
    TrapPlacement:Exec
    ShoulderLauncher:Exec
    AbilityDecoTool:Exec
    Crossbow:Exec
    C4:Exec
    RemoteControl:Exec
    DualWield:Exec
    AR_BullPup:Exec
    AR_ForwardGrip:Exec
    MedPackPaddles:Exec
    SMG_P90:Exec
    AR_DrumGun:Exec
    Consumable_Small:Exec
    Consumable_Large:Exec
    Balloon:Exec
    MountedTurret:Exec
    CreativeTool:Exec
    ExplosiveBow:Exec
    AshtonIndigo:Exec
    AshtonChicago:Exec
    MeleeDualWield:Exec
    Unarmed:Exec
   )
   Branch(execute=[@n5.AssaultRifle, @n5.Crossbow, @n5.GrenadeLauncher, @n5.Rifle, @n5.RocketLauncher, @n5.Shotgun, @n5.SniperRifle, @n5.TacticalShotgun])
   > Set isAimingRifle(isAimingRifle=true)
   Branch(execute=[@n5.MachinePistol, @n5.Pistol])
   > Set isAimingPistol(isAimingPistol=1)
   Set Skydiving(execute=@n1.then_1, Skydiving:bool=OR Boolean(A=Get bIsSkyDiving(), B=Get bIsSkydiveDiveMode())) -> (Output_Get:bool)
   Set JumpingFalling(execute=@n1.then_2, JumpingFalling:bool=OR Boolean(A=Get bIsJumping(), B=Get bIsFalling())) -> (Output_Get:bool)
   Set Crouching(execute=@n1.then_3, Crouching:bool=Get bIsCrouching()) -> (Output_Get:bool)
   n6 = Gravity Override State Function DEPRECATED(
    execute=@n1.then_4
    CharacterPartAnimInstance=Self-Reference()
    GameTimeInSeconds=GetGameTimeInSeconds()
    GravityOverrideParams=Get Gravity Override Params()
   )
   > Set Gravity_Override_FE(Gravity_Override_FE:Vector=@n6) -> (Output_Get:Vector)
   "To String (Float)"(InDouble=0)
   PrintString!(execute, InString=Hello, bPrintToScreen, bPrintToLog, TextColor, Duration, Key) -> (then)
  ]}
]
variables [
 Arm Side Lift L {type=number}
 Crouching {type=boolean}
 Gravity Override Params {
  type=GravityOverrideParamsStruct
  default {
   skydive_DivingDown {gravityOverride=<0 0 100>}
   skydive_FloatingDown {gravityOverride=<0 0 600>}
   skydive_GoingUp {gravityOverride=<0 0 150>}
   swimming {gravityOverride=<0 0 0>}
  }
 }
 Gravity_Override_FE {type=Vector}
 isAimingPistol {type=number}
 isAimingRifle {type=boolean}
 JumpingFalling {type=boolean}
 Skydiving {type=boolean}
]
