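# Physics asset for the Frey part of the medium male Snowboarder cosmetic body, giving its dynamic head bones physics bodies.
# Capsules cover the upper spine, clavicles, neck and head, linked in a chain by constraints with free swing and twist.
# Four small identical capsules on the left, right, top and back-hat dyn bones each constrain to the head with narrow angular limits, tightest on the back of the hat.
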PhysicsAsset /BRCosmetics/Characters/Player/Male/Medium/Bodies/M_MED_Snowboarder/Meshes/Parts/M_MED_Snowboarder_Frey_Physics
bodies {
 clavicle_l [{shapeType=Capsule, length=6.9452, radius=7.5074, center=<7.8663 -0.0001 0.5473>, rotation=<rot -86.0204 -0.0002 0.0002>}]
 clavicle_r [{shapeType=Capsule, length=7.4366, radius=7.016, center=<-10.1967 0 -1.339>, rotation=<rot -83.5322 -179.9998 179.9998>}]
 dyn_back_hat [{shapeType=Capsule, length=3, radius=3, center=<-4.2598 -0.0001 0>, rotation=<rot 90 0 0>}]
 dyn_l_frey [{shapeType=Capsule, length=3, radius=3, center=<-4.2598 -0.0001 0>, rotation=<rot 90 0 0>}]
 dyn_r_frey [{shapeType=Capsule, length=3, radius=3, center=<-4.2598 -0.0001 0>, rotation=<rot 90 0 0>}]
 dyn_top_frey [{shapeType=Capsule, length=3, radius=3, center=<-4.2598 -0.0001 0>, rotation=<rot 90 0 0>}]
 head [{shapeType=Capsule, length=7.9078, radius=8.9852, rotation=<rot 90 0 0>}]
 neck_01 [{shapeType=Capsule, length=4.8937, radius=5.117}]
 neck_02 [{shapeType=Capsule, length=4.8931, radius=5.117}]
 spine_04 [{shapeType=Capsule, length=24.8676, radius=12.1476}]
 spine_05 [{shapeType=Capsule, length=20.8747, radius=11.7474}]
}
constraints [
 {
  bone1Name=spine_05
  bone2Name=spine_04
  swing1LimitDegrees=45
  swing1Motion=Free
  swing2LimitDegrees=45
  swing2Motion=Free
  twistLimitDegrees=45
  twistMotion=Free
 }
 {
  bone1Name=clavicle_l
  bone2Name=spine_05
  swing1LimitDegrees=45
  swing1Motion=Free
  swing2LimitDegrees=45
  swing2Motion=Free
  twistLimitDegrees=45
  twistMotion=Free
 }
 {
  bone1Name=clavicle_r
  bone2Name=spine_05
  swing1LimitDegrees=45
  swing1Motion=Free
  swing2LimitDegrees=45
  swing2Motion=Free
  twistLimitDegrees=45
  twistMotion=Free
 }
 {
  bone1Name=neck_01
  bone2Name=spine_05
  swing1LimitDegrees=45
  swing1Motion=Free
  swing2LimitDegrees=45
  swing2Motion=Free
  twistLimitDegrees=45
  twistMotion=Free
 }
 {
  bone1Name=neck_02
  bone2Name=neck_01
  swing1LimitDegrees=45
  swing1Motion=Free
  swing2LimitDegrees=45
  swing2Motion=Free
  twistLimitDegrees=45
  twistMotion=Free
 }
 {
  bone1Name=head
  bone2Name=neck_02
  swing1LimitDegrees=45
  swing1Motion=Free
  swing2LimitDegrees=45
  swing2Motion=Free
  twistLimitDegrees=45
  twistMotion=Free
 }
 {bone1Name=dyn_r_frey, bone2Name=head, swing1LimitDegrees=30, swing2LimitDegrees=30, twistLimitDegrees=30}
 {bone1Name=dyn_l_frey, bone2Name=head, swing1LimitDegrees=30, swing2LimitDegrees=30, twistLimitDegrees=30}
 {bone1Name=dyn_top_frey, bone2Name=head, swing1LimitDegrees=30, swing2LimitDegrees=30, twistLimitDegrees=30}
 {bone1Name=dyn_back_hat, bone2Name=head, swing1LimitDegrees=10, swing2LimitDegrees=10, twistLimitDegrees=10}
]
